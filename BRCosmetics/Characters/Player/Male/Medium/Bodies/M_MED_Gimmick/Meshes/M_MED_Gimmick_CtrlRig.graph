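# Control Rig for the male Gimmick body that procedurally moves the six coat lapel bones as the arms move.
# Spherical pose readers on each upper arm, relative to spine_04, weight a shared function that adds local translate, rotate and scale offsets per bone.
# An extra right-clavicle reader further rotates two right lapel bones (spelled "rapel"), with no left equivalent.
# Debug axes are drawn before and after each side's corrections.
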ControlRigBlueprint /BRCosmetics/Characters/Player/Male/Medium/Bodies/M_MED_Gimmick/Meshes/M_MED_Gimmick_CtrlRig
graphs [
 ContainedGraph {nodes [
   n1 = RigVMFunctionEntryNode()
   RigUnit_GetTransform = &RigUnit_GetTransform(Space=LocalSpace, bInitial=false, Item=@n1.Joint)
   n2 = &RigVMFunction_MathQuaternionToEuler(RotationOrder=ZYX, Value=@RigUnit_GetTransform.Transform.Rotation)
   > RigUnit_SetTransform = &RigUnit_SetTransform(
    Space=LocalSpace
    Value=(Rotation=(X=0.0132,Y=0.4099,Z=-0.0222,W=0.9118),Translation=(X=23.6045,Y=-6.37,Z=140.1115),Scale3D=(X=1,Y=1,Z=1))
    bInitial=false
    Item=@n1.Joint
    Weight=@n1.Weight
    Value.Translation.Z=&RigVMFunction_MathFloatAdd(B=@n1.Translate Z, A=@RigUnit_GetTransform.Transform.Translation.Z) {template="Add::Execute(in A,in B,out Result)"}.Result
    Value.Scale3D.Z=&RigVMFunction_MathFloatAdd(B=@n1.Scale Z, A=@RigUnit_GetTransform.Transform.Scale3D.Z) {template="Add::Execute(in A,in B,out Result)"}.Result
    Value.Translation.Y=&RigVMFunction_MathFloatAdd(B=@n1.Translate Y, A=@RigUnit_GetTransform.Transform.Translation.Y) {template="Add::Execute(in A,in B,out Result)"}.Result
    Value.Scale3D.Y=&RigVMFunction_MathFloatAdd(B=@n1.Scale Y, A=@RigUnit_GetTransform.Transform.Scale3D.Y) {template="Add::Execute(in A,in B,out Result)"}.Result
    Value.Translation.X=&RigVMFunction_MathFloatAdd(B=@n1.Translate X, A=@RigUnit_GetTransform.Transform.Translation.X) {template="Add::Execute(in A,in B,out Result)"}.Result
    Value.Scale3D.X=&RigVMFunction_MathFloatAdd(B=@n1.Scale X, A=@RigUnit_GetTransform.Transform.Scale3D.X) {template="Add::Execute(in A,in B,out Result)"}.Result
    Value.Rotation=&RigVMFunction_MathQuaternionFromEuler(
     Euler=(X=0,Y=0,Z=0)
     Euler.X=&RigVMFunction_MathFloatAdd(B=@n1.Rotate X, A=@n2.Result.X) {template="Add::Execute(in A,in B,out Result)"}.Result
     Euler.Y=&RigVMFunction_MathFloatAdd(B=@n1.Rotate Y, A=@n2.Result.Y) {template="Add::Execute(in A,in B,out Result)"}.Result
     Euler.Z=&RigVMFunction_MathFloatAdd(B=@n1.Rotate Z, A=@n2.Result.Z) {template="Add::Execute(in A,in B,out Result)"}.Result
    ).Result
   )
   > RigVMFunctionReturnNode()
  ]}
 RigVMFunctionLibrary {nodes [Set Transforms for Bone = RigVMCollapseNode(
    ExecuteContext=()
    Joint=()
    Rotate X=0
    Rotate Y=0
    Rotate Z=0
    Scale X=0
    Scale Y=0
    Scale Z=0
    Translate X=0
    Translate Y=0
    Translate Z=0
    Weight=1
   )]}
 RigVMModel {nodes [
   RigUnit_BeginExecution = &RigUnit_BeginExecution()
   /Script/ControlRig.RigUnit_DebugTransformMutableItemSpace(
    Color=(R=1,G=0,B=0,A=1)
    DebugDrawSettings=(DepthPriority=SDPG_Foreground,Lifetime=-1)
    Mode=Axes
    Scale=5
    Space=(Type=Bone)
    Thickness=0.1
    WorldOffset=(Rotation=(X=0,Y=0,Z=0,W=1),Translation=(X=0,Y=0,Z=0),Scale3D=(X=1,Y=1,Z=1))
    bEnabled=True
    ExecutePin=@RigUnit_BeginExecution.ExecutePin
    Transform=&RigUnit_GetTransform(Item=(Type=Bone,Name="dyn_lapel_l")).Transform
   ) {is=struct RigVMUnitNode}
   > &RigUnit_DebugTransformMutableItemSpace(
    Color=(R=1,G=0,B=0,A=1)
    DebugDrawSettings=(DepthPriority=SDPG_Foreground,Lifetime=-1)
    Mode=Axes
    Scale=5
    Space=(Type=Bone)
    Thickness=0.1
    WorldOffset=(Rotation=(X=0,Y=0,Z=0,W=1),Translation=(X=0,Y=0,Z=0),Scale3D=(X=1,Y=1,Z=1))
    bEnabled=True
    Transform=&RigUnit_GetTransform(Item=(Type=Bone,Name="dyn_lapel_low_l")).Transform
   )
   > &RigUnit_DebugTransformMutableItemSpace(
    Color=(R=1,G=0,B=0,A=1)
    DebugDrawSettings=(DepthPriority=SDPG_Foreground,Lifetime=-1)
    Mode=Axes
    Scale=5
    Space=(Type=Bone)
    Thickness=0.1
    WorldOffset=(Rotation=(X=0,Y=0,Z=0,W=1),Translation=(X=0,Y=0,Z=0),Scale3D=(X=1,Y=1,Z=1))
    bEnabled=True
    Transform=&RigUnit_GetTransform(Item=(Type=Bone,Name="dyn_lapel_high_l")).Transform
   )
   > n1 = &RigUnit_SphericalPoseReader(
    ActiveRegionScaleFactors=(PositiveWidth=0.6,NegativeWidth=0.6,PositiveHeight=1,NegativeHeight=1)
    ActiveRegionSize=0.15
    Debug=(bDrawDebug=false,bDraw2D=false,bDrawLocalAxes=true,DebugScale=25,DebugSegments=20,DebugThickness=0.25)
    DriverAxis=(X=1,Y=0,Z=0)
    DriverItem=(Type=Bone,Name="upperarm_l")
    FalloffRegionScaleFactors=(PositiveWidth=0.6,NegativeWidth=0.6,PositiveHeight=1,NegativeHeight=1)
    FalloffSize=0.45
    OptionalParentItem=(Type=Bone,Name="spine_04")
    RotationOffset=(X=40,Y=25,Z=-40)
   )
   > Set Transforms for Bone = RigVMFunctionReferenceNode(
    Joint=(Type=Bone,Name="dyn_lapel_high_l")
    Rotate X=0
    Rotate Y=-70
    Rotate Z=0
    Scale X=0
    Scale Y=0
    Scale Z=0
    Translate X=-2
    Translate Y=0
    Translate Z=1
    Weight=@n1.OutputParam
   )
   > n2 = &RigUnit_SphericalPoseReader(
    ActiveRegionScaleFactors=(PositiveWidth=0.5,NegativeWidth=1,PositiveHeight=1,NegativeHeight=2)
    ActiveRegionSize=0.08
    Debug=(bDrawDebug=false,bDraw2D=false,bDrawLocalAxes=true,DebugScale=25,DebugSegments=20,DebugThickness=0.25)
    DriverAxis=(X=1,Y=0,Z=0)
    DriverItem=(Type=Bone,Name="upperarm_l")
    FalloffRegionScaleFactors=(PositiveWidth=0.5,NegativeWidth=1,PositiveHeight=1,NegativeHeight=1)
    FalloffSize=0.45
    OptionalParentItem=(Type=Bone,Name="spine_04")
    RotationOffset=(X=80,Y=30,Z=-25)
   )
   > Set Transforms for Bone_1 = RigVMFunctionReferenceNode(
    Joint=(Type=Bone,Name="dyn_lapel_l")
    Rotate X=0
    Rotate Y=-65
    Rotate Z=0
    Scale X=0
    Scale Y=0
    Scale Z=0
    Translate X=-1
    Translate Y=0
    Translate Z=1
    Weight=@n2.OutputParam
   )
   > n3 = &RigUnit_SphericalPoseReader(
    ActiveRegionScaleFactors=(PositiveWidth=1,NegativeWidth=1,PositiveHeight=3,NegativeHeight=2)
    ActiveRegionSize=0.05
    Debug=(bDrawDebug=false,bDraw2D=false,bDrawLocalAxes=true,DebugScale=25,DebugSegments=20,DebugThickness=0.25)
    DriverAxis=(X=1,Y=0,Z=0)
    DriverItem=(Type=Bone,Name="upperarm_l")
    FalloffRegionScaleFactors=(PositiveWidth=0.5,NegativeWidth=0.5,PositiveHeight=2,NegativeHeight=2)
    FalloffSize=0.3
    OptionalParentItem=(Type=Bone,Name="spine_04")
    RotationOffset=(X=110,Y=30,Z=-30)
   )
   > Set Transforms for Bone_1_1 = RigVMFunctionReferenceNode(
    Joint=(Type=Bone,Name="dyn_lapel_low_l")
    Rotate X=0
    Rotate Y=-40
    Rotate Z=0
    Scale X=0
    Scale Y=0
    Scale Z=0
    Translate X=-1
    Translate Y=0
    Translate Z=0
    Weight=@n3.OutputParam
   )
   > &RigUnit_DebugTransformMutableItemSpace(
    Color=(R=1,G=0,B=0,A=1)
    DebugDrawSettings=(DepthPriority=SDPG_Foreground,Lifetime=-1)
    Mode=Axes
    Scale=3
    Space=(Type=Bone)
    Thickness=0.3
    WorldOffset=(Rotation=(X=0,Y=0,Z=0,W=1),Translation=(X=0,Y=0,Z=0),Scale3D=(X=1,Y=1,Z=1))
    bEnabled=True
    Transform=&RigUnit_GetTransform(Item=(Type=Bone,Name="dyn_lapel_l")).Transform
   )
   > &RigUnit_DebugTransformMutableItemSpace(
    Color=(R=1,G=0,B=0,A=1)
    DebugDrawSettings=(DepthPriority=SDPG_Foreground,Lifetime=-1)
    Mode=Axes
    Scale=3
    Space=(Type=Bone)
    Thickness=0.3
    WorldOffset=(Rotation=(X=0,Y=0,Z=0,W=1),Translation=(X=0,Y=0,Z=0),Scale3D=(X=1,Y=1,Z=1))
    bEnabled=True
    Transform=&RigUnit_GetTransform(Item=(Type=Bone,Name="dyn_lapel_low_l")).Transform
   )
   > &RigUnit_DebugTransformMutableItemSpace(
    Color=(R=1,G=0,B=0,A=1)
    DebugDrawSettings=(DepthPriority=SDPG_Foreground,Lifetime=-1)
    Mode=Axes
    Scale=3
    Space=(Type=Bone)
    Thickness=0.3
    WorldOffset=(Rotation=(X=0,Y=0,Z=0,W=1),Translation=(X=0,Y=0,Z=0),Scale3D=(X=1,Y=1,Z=1))
    bEnabled=True
    Transform=&RigUnit_GetTransform(Item=(Type=Bone,Name="dyn_lapel_high_l")).Transform
   )
   > &RigUnit_DebugTransformMutableItemSpace(
    Color=(R=1,G=0,B=0,A=1)
    DebugDrawSettings=(DepthPriority=SDPG_Foreground,Lifetime=-1)
    Mode=Axes
    Scale=5
    Space=(Type=Bone)
    Thickness=0.1
    WorldOffset=(Rotation=(X=0,Y=0,Z=0,W=1),Translation=(X=0,Y=0,Z=0),Scale3D=(X=1,Y=1,Z=1))
    bEnabled=True
    Transform=&RigUnit_GetTransform(Item=(Type=Bone,Name="dyn_rapel_r")).Transform
   )
   > &RigUnit_DebugTransformMutableItemSpace(
    Color=(R=1,G=0,B=0,A=1)
    DebugDrawSettings=(DepthPriority=SDPG_Foreground,Lifetime=-1)
    Mode=Axes
    Scale=5
    Space=(Type=Bone)
    Thickness=0.1
    WorldOffset=(Rotation=(X=0,Y=0,Z=0,W=1),Translation=(X=0,Y=0,Z=0),Scale3D=(X=1,Y=1,Z=1))
    bEnabled=True
    Transform=&RigUnit_GetTransform(Item=(Type=Bone,Name="dyn_rapel_row_r")).Transform
   )
   > &RigUnit_DebugTransformMutableItemSpace(
    Color=(R=1,G=0,B=0,A=1)
    DebugDrawSettings=(DepthPriority=SDPG_Foreground,Lifetime=-1)
    Mode=Axes
    Scale=5
    Space=(Type=Bone)
    Thickness=0.1
    WorldOffset=(Rotation=(X=0,Y=0,Z=0,W=1),Translation=(X=0,Y=0,Z=0),Scale3D=(X=1,Y=1,Z=1))
    bEnabled=True
    Transform=&RigUnit_GetTransform(Item=(Type=Bone,Name="dyn_rapel_high_r")).Transform
   )
   > n4 = &RigUnit_SphericalPoseReader(
    ActiveRegionScaleFactors=(PositiveWidth=0.6,NegativeWidth=0.6,PositiveHeight=1,NegativeHeight=1)
    ActiveRegionSize=0.15
    Debug=(bDrawDebug=false,bDraw2D=false,bDrawLocalAxes=true,DebugScale=25,DebugSegments=20,DebugThickness=0.25)
    DriverAxis=(X=-1,Y=0,Z=0)
    DriverItem=(Type=Bone,Name="upperarm_r")
    FalloffRegionScaleFactors=(PositiveWidth=0.6,NegativeWidth=0.6,PositiveHeight=2,NegativeHeight=1)
    FalloffSize=0.45
    FlipHeightScaling=false
    FlipWidthScaling=true
    OptionalParentItem=(Type=Bone,Name="spine_04")
    RotationOffset=(X=-100,Y=25,Z=40)
   )
   > Set Transforms for Bone_1_1_1 = RigVMFunctionReferenceNode(
    Joint=(Type=Bone,Name="dyn_rapel_high_r")
    Rotate X=0
    Rotate Y=-70
    Rotate Z=0
    Scale X=0
    Scale Y=0
    Scale Z=0
    Translate X=0
    Translate Y=0
    Translate Z=0
    Weight=@n4.OutputParam
   )
   > n5 = &RigUnit_SphericalPoseReader(
    ActiveRegionScaleFactors=(PositiveWidth=0.5,NegativeWidth=1,PositiveHeight=1.8,NegativeHeight=2)
    ActiveRegionSize=0.08
    Debug=(bDrawDebug=false,bDraw2D=false,bDrawLocalAxes=true,DebugScale=25,DebugSegments=20,DebugThickness=0.25)
    DriverAxis=(X=-1,Y=0,Z=0)
    DriverItem=(Type=Bone,Name="upperarm_r")
    FalloffRegionScaleFactors=(PositiveWidth=0.5,NegativeWidth=1,PositiveHeight=2,NegativeHeight=1)
    FalloffSize=0.45
    FlipHeightScaling=false
    FlipWidthScaling=true
    OptionalParentItem=(Type=Bone,Name="spine_04")
    RotationOffset=(X=-80,Y=40,Z=25)
   )
   > Set Transforms for Bone_1_1_1_1 = RigVMFunctionReferenceNode(
    Joint=(Type=Bone,Name="dyn_rapel_r")
    Rotate X=0
    Rotate Y=-65
    Rotate Z=0
    Scale X=-0.3
    Scale Y=0
    Scale Z=0
    Translate X=-3
    Translate Y=0
    Translate Z=2
    Weight=@n5.OutputParam
   )
   > n6 = &RigUnit_SphericalPoseReader(
    ActiveRegionScaleFactors=(PositiveWidth=1,NegativeWidth=1,PositiveHeight=2,NegativeHeight=1)
    ActiveRegionSize=0.05
    Debug=(bDrawDebug=false,bDraw2D=false,bDrawLocalAxes=true,DebugScale=25,DebugSegments=20,DebugThickness=0.25)
    DriverAxis=(X=-1,Y=0,Z=0)
    DriverItem=(Type=Bone,Name="upperarm_r")
    FalloffRegionScaleFactors=(PositiveWidth=0.5,NegativeWidth=0.5,PositiveHeight=1,NegativeHeight=1)
    FalloffSize=0.3
    FlipWidthScaling=true
    OptionalParentItem=(Type=Bone,Name="spine_04")
    RotationOffset=(X=-50,Y=35,Z=20)
   )
   > Set Transforms for Bone_1_1_1_1_1 = RigVMFunctionReferenceNode(
    Joint=(Type=Bone,Name="dyn_rapel_row_r")
    Rotate X=0
    Rotate Y=-60
    Rotate Z=0
    Scale X=0
    Scale Y=0
    Scale Z=0
    Translate X=-1
    Translate Y=0
    Translate Z=3
    Weight=@n6.OutputParam
   )
   > n7 = &RigUnit_SphericalPoseReader(
    ActiveRegionScaleFactors=(PositiveWidth=1,NegativeWidth=1,PositiveHeight=1,NegativeHeight=1)
    ActiveRegionSize=0.02
    Debug=(bDrawDebug=false,bDraw2D=false,bDrawLocalAxes=true,DebugScale=25,DebugSegments=12,DebugThickness=0.15)
    DriverAxis=(X=-1,Y=0,Z=0)
    DriverItem=(Type=Bone,Name="clavicle_r")
    FalloffRegionScaleFactors=(PositiveWidth=1,NegativeWidth=1,PositiveHeight=1,NegativeHeight=1)
    FalloffSize=0.15
    FlipWidthScaling=false
    OptionalParentItem=(Type=Bone,Name="spine_05")
    RotationOffset=(X=0,Y=130,Z=20)
   )
   > Set Transforms for Bone_1_1_1_1_1_1 = RigVMFunctionReferenceNode(
    Joint=(Type=Bone,Name="dyn_rapel_high_r")
    Rotate X=0
    Rotate Y=-80
    Rotate Z=0
    Scale X=0
    Scale Y=0
    Scale Z=0
    Translate X=0
    Translate Y=0
    Translate Z=0
    Weight=@n7.OutputParam
   )
   > Set Transforms for Bone_1_1_1_1_1_1_1 = RigVMFunctionReferenceNode(
    Joint=(Type=Bone,Name="dyn_rapel_r")
    Rotate X=0
    Rotate Y=-40
    Rotate Z=0
    Scale X=0
    Scale Y=0
    Scale Z=0
    Translate X=0
    Translate Y=0
    Translate Z=0
    Weight=@n7.OutputParam
   )
   > &RigUnit_DebugTransformMutableItemSpace(
    Color=(R=1,G=0,B=0,A=1)
    DebugDrawSettings=(DepthPriority=SDPG_Foreground,Lifetime=-1)
    Mode=Axes
    Scale=3
    Space=(Type=Bone)
    Thickness=0.3
    WorldOffset=(Rotation=(X=0,Y=0,Z=0,W=1),Translation=(X=0,Y=0,Z=0),Scale3D=(X=1,Y=1,Z=1))
    bEnabled=True
    Transform=&RigUnit_GetTransform(Item=(Type=Bone,Name="dyn_rapel_r")).Transform
   )
   > &RigUnit_DebugTransformMutableItemSpace(
    Color=(R=1,G=0,B=0,A=1)
    DebugDrawSettings=(DepthPriority=SDPG_Foreground,Lifetime=-1)
    Mode=Axes
    Scale=3
    Space=(Type=Bone)
    Thickness=0.3
    WorldOffset=(Rotation=(X=0,Y=0,Z=0,W=1),Translation=(X=0,Y=0,Z=0),Scale3D=(X=1,Y=1,Z=1))
    bEnabled=True
    Transform=&RigUnit_GetTransform(Item=(Type=Bone,Name="dyn_rapel_row_r")).Transform
   )
   > &RigUnit_DebugTransformMutableItemSpace(
    Color=(R=1,G=0,B=0,A=1)
    DebugDrawSettings=(DepthPriority=SDPG_Foreground,Lifetime=-1)
    Mode=Axes
    Scale=3
    Space=(Type=Bone)
    Thickness=0.3
    WorldOffset=(Rotation=(X=0,Y=0,Z=0,W=1),Translation=(X=0,Y=0,Z=0),Scale3D=(X=1,Y=1,Z=1))
    bEnabled=True
    Transform=&RigUnit_GetTransform(Item=(Type=Bone,Name="dyn_rapel_high_r")).Transform
   )
  ]}
]
hierarchy [
 root {loc=<0 0 0>, rot=<0 0 0 1>}
 pelvis {parent=root, loc=<0 -0.7071 93.1191>, rot=<0.0127 0.707 -0.0127 -0.707>}
 spine_01 {parent=pelvis, loc=<6.7494 -0.6388 0>, rot=<0 0 0.034 -0.9994>}
 spine_02 {parent=spine_01, loc=<7.2518 -0.035 0>, rot=<0 0 0.0238 -0.9997>}
 spine_03 {parent=spine_02, loc=<6.0104 0 0>, rot=<0 0 0.0149 -0.9999>}
 spine_04 {parent=spine_03, loc=<8.1024 0 0>, rot=<0 0 -0.0132 -0.9999>}
 spine_05 {parent=spine_04, loc=<10.0146 0 0>, rot=<0 0 -0.0015 -1>}
 clavicle_l {parent=spine_05, loc=<4.3575 -0.1466 -2.3223>, rot=<0.0465 -0.7289 -0.009 -0.6829>}
 upperarm_l {parent=clavicle_l, loc=<15.5111 0 0>, rot=<0.0226 -0.3234 -0.0088 -0.9459>}
 lowerarm_l {parent=upperarm_l, loc=<26.6514 0 0>, rot=<0.0391 -0.0099 -0.2453 -0.9686>}
 hand_l {parent=lowerarm_l, loc=<22.3491 0 0>, rot=<-0.7081 -0.0507 0.0975 -0.6975>}
 index_metacarpal_l {parent=hand_l, loc=<1.6535 0.0183 -2.96>, rot=<0.0158 -0.0124 -0.0173 -0.9996>}
 index_01_l {parent=index_metacarpal_l, loc=<8.3008 0.0357 -0.0524>, rot=<-0.0044 0.0485 0.0924 -0.9945>}
 index_02_l {parent=index_01_l, loc=<3.3587 -0.3004 0.0452>, rot=<-0.002 0.0027 0.3225 -0.9466>}
 index_03_l {parent=index_02_l, loc=<2.5578 -0.123 -0.0252>, rot=<-0.002 0.0067 0.2558 -0.9667>}
 middle_metacarpal_l {parent=hand_l, loc=<2.1856 -0.1725 -0.5984>, rot=<0.006 0.0168 -0.0329 -0.9993>}
 middle_01_l {parent=middle_metacarpal_l, loc=<7.3275 -0.2447 0.0129>, rot=<-0.0208 0.0042 0.0858 -0.9961>}
 middle_02_l {parent=middle_01_l, loc=<4.0123 0.1401 0.1175>, rot=<-0.0218 0.0258 0.3512 -0.9357>}
 middle_03_l {parent=middle_02_l, loc=<2.6077 0.0534 0.1373>, rot=<0.0082 0.0188 0.1708 -0.9851>}
 pinky_metacarpal_l {parent=hand_l, loc=<1.7616 -0.9753 3.3814>, rot=<-0.0397 0.0814 -0.069 -0.9935>}
 pinky_01_l {parent=pinky_metacarpal_l, loc=<7.1709 -0.4831 -0.1094>, rot=<-0.0654 -0.0035 0.2242 -0.9723>}
 pinky_02_l {parent=pinky_01_l, loc=<2.8274 0.0893 -0.1043>, rot=<0.0037 0.0011 0.2854 -0.9584>}
 pinky_03_l {parent=pinky_02_l, loc=<2.2046 0.0307 -0.0078>, rot=<-0.0002 0.0098 0.1445 -0.9895>}
 ring_metacarpal_l {parent=hand_l, loc=<2.2074 -0.3625 1.3638>, rot=<-0.0284 0.0722 -0.0458 -0.9959>}
 ring_01_l {parent=ring_metacarpal_l, loc=<7.0295 -0.6197 -0.1446>, rot=<-0.0489 -0.0252 0.1892 -0.9804>}
 ring_02_l {parent=ring_01_l, loc=<3.6174 0.4647 -0.1027>, rot=<0 0 0.2806 -0.9598>}
 ring_03_l {parent=ring_02_l, loc=<2.4152 0.114 0.0199>, rot=<0 0 0.1189 -0.9929>}
 thumb_01_l {parent=hand_l, loc=<1.3118 -0.3191 -2.459>, rot=<0.6546 -0.4042 0.0664 -0.6354>}
 thumb_02_l {parent=thumb_01_l, loc=<5.9336 -0.0448 0.0401>, rot=<-0.0037 0.0213 0.298 -0.9543>}
 thumb_03_l {parent=thumb_02_l, loc=<2.2223 0.3682 -0.0907>, rot=<0.0118 -0.013 0.0066 -0.9998>}
 lowerarm_twist_01_l {parent=lowerarm_l, loc=<17.8549 0 0>, rot=<0 0 0 1>}
 dyn_wrist_strap_l_01_MSTR {parent=lowerarm_twist_01_l, loc=<-4.8054 -5.4752 -0.5102>, rot=<-0.3978 -0.4408 0.5694 -0.5686>}
 dyn_wrist_strap_l_01 {parent=dyn_wrist_strap_l_01_MSTR, loc=<0.3 0 0>, rot=<0 0 0 1>}
 dyn_wrist_strap_l_02_MSTR {parent=lowerarm_twist_01_l, loc=<-1.2489 -5.3498 -0.6868>, rot=<-0.4151 -0.423 0.5827 -0.5561>}
 dyn_wrist_strap_l_02 {parent=dyn_wrist_strap_l_02_MSTR, loc=<0.3 0 0>, rot=<0 0 0 1>}
 lowerarm_twist_02_l {parent=lowerarm_l, loc=<9.6199 0 0>, rot=<0 0 0 1>}
 deform_elbow_bk_l {parent=lowerarm_l, loc=<-1.2853 -2.9274 -0.5607>, rot=<-0.0285 0.0335 -0.0024 -0.999>}
 upperarm_twist_01_l {parent=upperarm_l, loc=<0.0041 0 0>, rot=<0 0 0 1>}
 deform_deltoid_fr_l {parent=upperarm_twist_01_l, loc=<2.4511 0 0>, rot=<0 0 0 1>}
 upperarm_twist_02_l {parent=upperarm_l, loc=<13.7941 0 0>, rot=<0 0 0 1>}
 deform_elbow_fr_l {parent=upperarm_l, loc=<26.5048 0.1875 -0.5816>, rot=<0.0021 0.0156 -0.185 -0.9826>}
 deform_pectoral_fr_l {parent=clavicle_l, loc=<9.3905 0 0>, rot=<0 0 0 1>}
 dyn_collar_l_MSTR {parent=clavicle_l, loc=<7.4431 3.0944 8.6574>, rot=<0.2333 -0.0406 -0.0512 -0.9702>}
 dyn_collar_l {parent=dyn_collar_l_MSTR, loc=<0.3 0 0>, rot=<0 0 0 1>}
 clavicle_r {parent=spine_05, loc=<4.3575 -0.1466 2.3223>, rot=<0.7289 0.0465 -0.6829 0.009>}
 upperarm_r {parent=clavicle_r, loc=<-15.5111 0 0>, rot=<0.0226 -0.3234 -0.0088 -0.9459>}
 lowerarm_r {parent=upperarm_r, loc=<-26.6515 0 0>, rot=<0.0391 -0.0099 -0.2453 -0.9686>}
 hand_r {parent=lowerarm_r, loc=<-22.3492 0 0>, rot=<-0.7081 -0.0507 0.0975 -0.6975>}
 index_metacarpal_r {parent=hand_r, loc=<-1.6534 -0.0181 2.9601>, rot=<0.0158 -0.0124 -0.0173 -0.9996>}
 index_01_r {parent=index_metacarpal_r, loc=<-8.3008 -0.0357 0.0524>, rot=<-0.0044 0.0485 0.0924 -0.9945>}
 index_02_r {parent=index_01_r, loc=<-3.3589 0.3003 -0.0453>, rot=<-0.002 0.0027 0.3225 -0.9466>}
 index_03_r {parent=index_02_r, loc=<-2.5578 0.1231 0.0251>, rot=<-0.002 0.0067 0.2558 -0.9667>}
 middle_metacarpal_r {parent=hand_r, loc=<-2.1856 0.1726 0.5984>, rot=<0.006 0.0168 -0.0329 -0.9993>}
 middle_01_r {parent=middle_metacarpal_r, loc=<-7.3273 0.2447 -0.0129>, rot=<-0.0208 0.0042 0.0858 -0.9961>}
 middle_02_r {parent=middle_01_r, loc=<-4.0142 -0.1411 -0.07>, rot=<-0.0217 0.0259 0.3512 -0.9357>}
 middle_03_r {parent=middle_02_r, loc=<-2.6074 -0.0534 -0.1312>, rot=<0.0083 0.0188 0.1708 -0.9851>}
 pinky_metacarpal_r {parent=hand_r, loc=<-1.7617 0.9753 -3.3814>, rot=<-0.0397 0.0814 -0.069 -0.9935>}
 pinky_01_r {parent=pinky_metacarpal_r, loc=<-7.1707 0.4832 0.1094>, rot=<-0.0654 -0.0035 0.2242 -0.9723>}
 pinky_02_r {parent=pinky_01_r, loc=<-2.8274 -0.0892 0.1043>, rot=<0.0037 0.0011 0.2854 -0.9584>}
 pinky_03_r {parent=pinky_02_r, loc=<-2.2045 -0.031 0.0078>, rot=<-0.0002 0.0098 0.1445 -0.9895>}
 ring_metacarpal_r {parent=hand_r, loc=<-2.2073 0.3627 -1.3638>, rot=<-0.0284 0.0722 -0.0458 -0.9959>}
 ring_01_r {parent=ring_metacarpal_r, loc=<-7.0297 0.6197 0.1446>, rot=<-0.0489 -0.0252 0.1892 -0.9804>}
 ring_02_r {parent=ring_01_r, loc=<-3.6171 -0.4648 0.1027>, rot=<0 0 0.2806 -0.9598>}
 ring_03_r {parent=ring_02_r, loc=<-2.4153 -0.1141 -0.0199>, rot=<0 0 0.1189 -0.9929>}
 thumb_01_r {parent=hand_r, loc=<-1.3119 0.3192 2.459>, rot=<0.6546 -0.4042 0.0664 -0.6354>}
 thumb_02_r {parent=thumb_01_r, loc=<-5.9335 0.0447 -0.0401>, rot=<-0.0037 0.0213 0.298 -0.9543>}
 thumb_03_r {parent=thumb_02_r, loc=<-2.2223 -0.3682 0.0907>, rot=<0.0118 -0.013 0.0066 -0.9998>}
 lowerarm_twist_01_r {parent=lowerarm_r, loc=<-17.855 0 0>, rot=<0 0 0 1>}
 dyn_wrist_strap_r_01_MSTR {parent=lowerarm_twist_01_r, loc=<4.8056 5.4752 0.5101>, rot=<0.5694 0.5686 0.3978 -0.4408>}
 dyn_wrist_strap_r_01 {parent=dyn_wrist_strap_r_01_MSTR, loc=<0.3 0.0001 0>, rot=<0 0 0 1>}
 dyn_wrist_strap_r_02_MSTR {parent=lowerarm_twist_01_r, loc=<1.2487 5.3499 0.6872>, rot=<0.5827 0.5561 0.4151 -0.423>}
 dyn_wrist_strap_r_02 {parent=dyn_wrist_strap_r_02_MSTR, loc=<0.3 0.0001 0>, rot=<0 0 0 1>}
 lowerarm_twist_02_r {parent=lowerarm_r, loc=<-9.6199 0 0>, rot=<0 0 0 1>}
 deform_elbow_bk_r {parent=lowerarm_r, loc=<1.2852 2.9274 0.561>, rot=<-0.0285 0.0335 -0.0024 -0.999>}
 upperarm_twist_01_r {parent=upperarm_r, loc=<-0.0041 0 0>, rot=<0 0 0 1>}
 deform_deltoid_fr_r {parent=upperarm_twist_01_r, loc=<-2.4512 0 0>, rot=<0 0 0 1>}
 upperarm_twist_02_r {parent=upperarm_r, loc=<-13.7942 0 0>, rot=<0 0 0 1>}
 deform_elbow_fr_r {parent=upperarm_r, loc=<-26.5046 -0.1875 0.5813>, rot=<0.0021 0.0156 -0.185 -0.9826>}
 deform_pectoral_fr_r {parent=clavicle_r, loc=<-9.39 0 0>, rot=<0 0 0 1>}
 dyn_collar_r_MSTR {parent=clavicle_r, loc=<-7.4432 -3.0944 -8.657>, rot=<0.0512 -0.9702 0.2333 0.0406>}
 dyn_collar_r {parent=dyn_collar_r_MSTR, loc=<0.2998 0.0003 0.0006>, rot=<0 0 0 1>}
 neck_01 {parent=spine_05, loc=<11.3138 0.1529 0>, rot=<0 0 -0.2003 -0.9797>}
 neck_02 {parent=neck_01, loc=<4.5627 0 0>, rot=<0 0 0.0504 -0.9987>}
 head {parent=neck_02, loc=<5.3914 0 0>, rot=<0 0 0.115 -0.9934>}
 dyn_lapel_l_MSTR {parent=spine_05, loc=<2.6767 11.2744 -10.3489>, rot=<0.5166 -0.6602 -0.2961 -0.4577>}
 dyn_lapel_l {parent=dyn_lapel_l_MSTR, loc=<0.3 0 0>, rot=<0 0 0 1>}
 dyn_rapel_r_MSTR {parent=spine_05, loc=<2.6771 11.2745 10.3489>, rot=<0.4577 0.2961 0.6602 -0.5166>}
 dyn_rapel_r {parent=dyn_rapel_r_MSTR, loc=<0.3 0 0>, rot=<0 0 0 1>}
 dyn_lapel_high_l_MSTR {parent=spine_05, loc=<8.2754 8.7289 -9.4263>, rot=<0.4009 -0.7029 -0.2125 -0.5478>}
 dyn_lapel_high_l {parent=dyn_lapel_high_l_MSTR, loc=<0.3 0 0>, rot=<0 0 0 1>}
 dyn_rapel_high_r_MSTR {parent=spine_05, loc=<8.2751 8.7289 9.4263>, rot=<0.5478 0.2125 0.7029 -0.4009>}
 dyn_rapel_high_r {parent=dyn_rapel_high_r_MSTR, loc=<0.3 0.0003 0.0003>, rot=<0 0 0 1>}
 dyn_coat_l_MSTR {parent=spine_04, loc=<2.951 10.5522 -12.2799>, rot=<0.0881 -0.8693 -0.4851 0.0335>}
 dyn_coat_l_01 {parent=dyn_coat_l_MSTR, loc=<0.3 0 0>, rot=<0 0 0 1>}
 dyn_coat_l_02 {parent=dyn_coat_l_01, loc=<11.567 0 0>, rot=<-0.2943 0.076 0.003 -0.9527>}
 dyn_coat_l_03 {parent=dyn_coat_l_02, loc=<10.5117 -1.2787 -1.2312>, rot=<-0.3806 0.0292 0.0472 -0.9231>}
 dyn_coat_l_04 {parent=dyn_coat_l_03, loc=<15.8782 0 0>, rot=<0 -0.0045 0 -1>}
 dyn_coat_l_05 {parent=dyn_coat_l_04, loc=<20.2787 0 0.0147>, rot=<0 -0.016 0 -0.9999>}
 dyn_coat_l_06 {parent=dyn_coat_l_05, loc=<18.03 0 0>, rot=<0 0 0 1>}
 dyn_belt_l_MSTR {parent=dyn_coat_l_03, loc=<-0.5655 -5.5244 2.3812>, rot=<0.4913 0.1139 -0.1845 -0.8436>}
 dyn_belt_l_01 {parent=dyn_belt_l_MSTR, loc=<0.3 0 0>, rot=<0 0 0 1>}
 dyn_belt_l_02 {parent=dyn_belt_l_01, loc=<8.8248 0 0>, rot=<-0.0087 -0.1458 0.0591 -0.9875>}
 dyn_belt_l_03 {parent=dyn_belt_l_02, loc=<9.3967 0 0>, rot=<0 0.0581 0 -0.9983>}
 dyn_belt_l_04 {parent=dyn_belt_l_03, loc=<11.0233 0 0>, rot=<0 0 0 1>}
 dyn_coat_r_MSTR {parent=spine_04, loc=<2.9511 10.5522 12.2799>, rot=<-0.0335 0.4851 0.8693 -0.0881>}
 dyn_coat_r_01 {parent=dyn_coat_r_MSTR, loc=<0.2997 0 -0.0001>, rot=<0 0 0 1>}
 dyn_coat_r_02 {parent=dyn_coat_r_01, loc=<11.5674 0 0>, rot=<0.2943 0.076 -0.003 -0.9527>}
 dyn_coat_r_03 {parent=dyn_coat_r_02, loc=<10.5118 1.2787 -1.2312>, rot=<0.3806 0.0292 -0.0472 -0.9231>}
 dyn_coat_r_04 {parent=dyn_coat_r_03, loc=<15.8781 0 0>, rot=<0 -0.0045 0 -1>}
 dyn_coat_r_05 {parent=dyn_coat_r_04, loc=<20.2787 0 0.0148>, rot=<0 -0.016 0 -0.9999>}
 dyn_coat_r_06 {parent=dyn_coat_r_05, loc=<18.0299 0 0>, rot=<0 0 0 -1>}
 dyn_belt_r_MSTR {parent=dyn_coat_r_03, loc=<-0.5657 5.5244 2.3813>, rot=<-0.4913 0.1139 0.1845 -0.8436>}
 dyn_belt_r_01 {parent=dyn_belt_r_MSTR, loc=<0.3005 0 -0.0003>, rot=<0 0 0 1>}
 dyn_belt_r_02 {parent=dyn_belt_r_01, loc=<8.8244 0 0.0002>, rot=<0.0087 -0.1458 -0.0591 -0.9875>}
 dyn_belt_r_03 {parent=dyn_belt_r_02, loc=<9.3968 0 0>, rot=<0 0.0581 0 -0.9983>}
 dyn_belt_r_04 {parent=dyn_belt_r_03, loc=<11.0232 0 0.0001>, rot=<0 0 0 -1>}
 dyn_lapel_low_l_MSTR {parent=spine_04, loc=<6.4254 12.1795 -10.2348>, rot=<0.6014 -0.5966 -0.3815 -0.37>}
 dyn_lapel_low_l {parent=dyn_lapel_low_l_MSTR, loc=<0.3 0 0>, rot=<0 0 0 1>}
 dyn_rapel_row_r_MSTR {parent=spine_04, loc=<6.4249 12.1794 10.2348>, rot=<0.37 0.3815 0.5966 -0.6014>}
 dyn_rapel_row_r {parent=dyn_rapel_row_r_MSTR, loc=<0.2999 0.0007 0>, rot=<0 0 0 1>}
 dyn_coat_bk_l_MSTR {parent=spine_02, loc=<-0.6086 -5.4204 -9.521>, rot=<0.0903 -0.8963 0.413 -0.1338>}
 dyn_coat_bk_l_01 {parent=dyn_coat_bk_l_MSTR, loc=<0.3 0 0>, rot=<0 0 0 1>}
 dyn_coat_bk_l_02 {parent=dyn_coat_bk_l_01, loc=<11.0848 0 0>, rot=<0 -0.0641 0 -0.9979>}
 dyn_coat_bk_l_04 {parent=dyn_coat_bk_l_02, loc=<14.588 0 0.0559>, rot=<0 -0.0116 0 -0.9999>}
 dyn_coat_bk_l_05 {parent=dyn_coat_bk_l_04, loc=<13.6078 0 0>, rot=<0 0.036 0 -0.9994>}
 dyn_coat_bk_l_06 {parent=dyn_coat_bk_l_05, loc=<14.9961 0 0>, rot=<0 0 0 1>}
 dyn_coat_bk_MSTR {parent=spine_02, loc=<-1.4061 -9.5095 0>, rot=<0.1192 -0.697 0.697 -0.1192>}
 dyn_coat_bk_01 {parent=dyn_coat_bk_MSTR, loc=<0.3 0 0>, rot=<0 0 0 1>}
 dyn_coat_bk_02 {parent=dyn_coat_bk_01, loc=<10.4809 0 0>, rot=<0 -0.064 0 -0.998>}
 dyn_coat_bk_04 {parent=dyn_coat_bk_02, loc=<12.8611 0 0.2968>, rot=<0 -0.014 0 -0.9999>}
 dyn_coat_bk_split_l_01 {parent=dyn_coat_bk_04, loc=<15.1935 2.1283 -0.4447>, rot=<0 -0.0025 0 -1>}
 dyn_coat_bk_split_l_02 {parent=dyn_coat_bk_split_l_01, loc=<13.982 0 0>, rot=<0 0 0 1>}
 dyn_coat_bk_split_r_01 {parent=dyn_coat_bk_04, loc=<15.1935 -2.1283 -0.4447>, rot=<0 -0.0025 0 -1>}
 dyn_coat_bk_split_r_02 {parent=dyn_coat_bk_split_r_01, loc=<13.982 0 0.0001>, rot=<0 0 0 1>}
 dyn_coat_bk_r_MSTR {parent=spine_02, loc=<-0.6089 -5.4204 9.521>, rot=<0.1338 -0.413 0.8963 -0.0903>}
 dyn_coat_bk_r_01 {parent=dyn_coat_bk_r_MSTR, loc=<0.3 0 0>, rot=<0 0 0 1>}
 dyn_coat_bk_r_02 {parent=dyn_coat_bk_r_01, loc=<11.0844 0 0.0001>, rot=<0 -0.0641 0 -0.9979>}
 dyn_coat_bk_r_04 {parent=dyn_coat_bk_r_02, loc=<14.588 0 0.0559>, rot=<0 -0.0116 0 -0.9999>}
 dyn_coat_bk_r_05 {parent=dyn_coat_bk_r_04, loc=<13.6079 0 -0.0001>, rot=<0 0.036 0 -0.9994>}
 dyn_coat_bk_r_06 {parent=dyn_coat_bk_r_05, loc=<14.9961 -0.0001 0>, rot=<0 0 0 1>}
 thigh_l {parent=pelvis, loc=<-2.5693 -0.0109 -7.7501>, rot=<0.0012 0.0572 0.0155 -0.9982>}
 calf_l {parent=thigh_l, loc=<-39.0484 0 0>, rot=<0.0006 -0.0161 -0.0391 -0.9991>}
 calf_twist_01_l {parent=calf_l, loc=<-35.813 -0.0967 0.0397>, rot=<0.0237 0.0095 -0.0035 -0.9997>}
 calf_twist_02_l {parent=calf_l, loc=<-18.8064 -0.0967 0.0397>, rot=<0.0237 0.0095 -0.0035 -0.9997>}
 foot_l {parent=calf_l, loc=<-42.561 0.1077 0.4292>, rot=<0.0418 -0.0455 0.0343 -0.9975>}
 ball_l {parent=foot_l, loc=<-6.1599 13.8264 0.6806>, rot=<0.0001 -0.0001 -0.7063 -0.7079>}
 deform_knee_bk_01_l {parent=calf_l, loc=<0.6022 -1.0607 0>, rot=<0 0 -0.0408 -0.9992>}
 deform_knee_fr_01_l {parent=calf_l, loc=<-0.2342 0 0>, rot=<0 0 0 1>}
 thigh_twist_01_l {parent=thigh_l, loc=<-21.6303 0 0>, rot=<0.0013 0 0 -1>}
 deform_thigh_bk_l {parent=thigh_twist_01_l, loc=<16.9514 0.2058 0.1808>, rot=<-0.0013 -0.0217 0 -0.9998>}
 deform_thigh_fr_l {parent=thigh_twist_01_l, loc=<21.5084 0.2281 -0.0598>, rot=<-0.0013 -0.0217 0 -0.9998>}
 thigh_r {parent=pelvis, loc=<-2.5691 -0.0109 7.7501>, rot=<0.0572 -0.0012 0.9982 0.0155>}
 calf_r {parent=thigh_r, loc=<39.0487 0 0>, rot=<0.0006 -0.0161 -0.0391 -0.9991>}
 calf_twist_01_r {parent=calf_r, loc=<35.8132 0.0967 -0.0397>, rot=<0.0237 0.0095 -0.0035 -0.9997>}
 calf_twist_02_r {parent=calf_r, loc=<18.8064 0.0967 -0.0397>, rot=<0.0237 0.0095 -0.0035 -0.9997>}
 foot_r {parent=calf_r, loc=<42.5612 -0.1077 -0.4292>, rot=<0.0418 -0.0455 0.0343 -0.9975>}
 ball_r {parent=foot_r, loc=<6.1598 -13.8264 -0.6806>, rot=<0.0001 -0.0001 -0.7063 -0.7079>}
 deform_knee_bk_01_r {parent=calf_r, loc=<-0.6022 1.0607 0>, rot=<0 0 -0.0408 -0.9992>}
 deform_knee_fr_01_r {parent=calf_r, loc=<0.2341 0 0>, rot=<0 0 0 1>}
 thigh_twist_01_r {parent=thigh_r, loc=<21.6303 0 0>, rot=<0.0013 0 0 -1>}
 deform_thigh_bk_r {parent=thigh_twist_01_r, loc=<-16.9512 -0.2058 -0.1808>, rot=<-0.0013 -0.0217 0 -0.9998>}
 deform_thigh_fr_r {parent=thigh_twist_01_r, loc=<-21.5083 -0.2281 0.0598>, rot=<-0.0013 -0.0217 0 -0.9998>}
]
