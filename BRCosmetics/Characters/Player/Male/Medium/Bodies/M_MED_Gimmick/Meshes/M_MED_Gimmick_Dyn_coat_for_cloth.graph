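# Physics asset for the Gimmick medium male body's dynamic coat, named for use with cloth simulation.
# Capsules cover the pelvis, spine and thighs, plus a left and right chain of three coat bones, each hung from a tiny master capsule on spine_03.
# Coat chain links have narrow twist and first swing limits with a wider second swing, while body joints and master attachments rotate freely.
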PhysicsAsset /BRCosmetics/Characters/Player/Male/Medium/Bodies/M_MED_Gimmick/Meshes/M_MED_Gimmick_Dyn_coat_for_cloth
bodies {
 dyn_coat_l_01 [{shapeType=Capsule, length=10.7561, radius=8, shapeName=dyn_coat_l_01_capsule, center=<4.667 -7.3674 1.8844>, rotation=<rot 90 0 0>}]
 dyn_coat_l_02 [{
   shapeType=Capsule
   length=9.094
   radius=11.1621
   shapeName=dyn_coat_l_01_capsule
   center=<5.44 -6.7379 8.2569>
   rotation=<rot 39.9131 -89.9998 -89.9998>
  }]
 dyn_coat_l_03 [{shapeType=Capsule, length=16.2561, radius=13, shapeName=dyn_coat_l_01_capsule, center=<10.3356 1.9842 11.368>, rotation=<rot 90 0 0>}]
 dyn_coat_l_MSTR [{shapeType=Capsule, length=0.505, radius=0.505, shapeName=dyn_coat_l_MSTR_capsule}]
 dyn_coat_r_01 [{shapeType=Capsule, length=10.7561, radius=8, shapeName=dyn_coat_l_01_capsule, center=<4.667 7.3674 1.8844>, rotation=<rot 90 0 0>}]
 dyn_coat_r_02 [{
   shapeType=Capsule
   length=12.2561
   radius=8
   shapeName=dyn_coat_l_01_capsule
   center=<5.44 4.2595 5.0146>
   rotation=<rot 39.9131 -89.9998 -89.9998>
  }]
 dyn_coat_r_03 [{shapeType=Capsule, length=21.2561, radius=8, shapeName=dyn_coat_l_01_capsule, center=<10.3356 0 6.8531>, rotation=<rot 90 0 0>}]
 dyn_coat_r_MSTR [{shapeType=Capsule, length=0.505, radius=0.505, shapeName=dyn_coat_l_MSTR_capsule}]
 pelvis [{shapeType=Capsule, length=8.136, radius=11.3521, shapeName=pelvis_capsule, center=<-1.6369 -2.7265 -0.0001>}]
 spine_01 [{shapeType=Capsule, length=10.0505, radius=11.3946, shapeName=pelvis_capsule, center=<-0.9991 -0.3721 0>}]
 spine_02 [{shapeType=Capsule, length=8.3482, radius=11.7708, shapeName=pelvis_capsule, center=<0.1451 -0.7547 0>}]
 spine_03 [{shapeType=Capsule, length=5.7572, radius=12.1816, shapeName=pelvis_capsule, center=<2.1769 0.258 0>}]
 thigh_l [{shapeType=Capsule, length=31.9975, radius=10.6235, shapeName=pelvis_capsule, center=<-19.9694 0.146 1.5609>, rotation=<rot 90 0 0>}]
 thigh_r [{shapeType=Capsule, length=31.9975, radius=10.6235, shapeName=pelvis_capsule, center=<19.9694 -0.146 -1.561>, rotation=<rot 90 0 0>}]
}
constraints [
 {bone1Name=dyn_coat_l_01, bone2Name=dyn_coat_l_MSTR, swing1LimitDegrees=10, swing2LimitDegrees=45, twistLimitDegrees=5}
 {bone1Name=dyn_coat_l_02, bone2Name=dyn_coat_l_01, swing1LimitDegrees=10, swing2LimitDegrees=45, twistLimitDegrees=5}
 {bone1Name=dyn_coat_l_03, bone2Name=dyn_coat_l_02, swing1LimitDegrees=10, swing2LimitDegrees=45, twistLimitDegrees=5}
 {bone1Name=dyn_coat_r_01, bone2Name=dyn_coat_r_MSTR, swing1LimitDegrees=10, swing2LimitDegrees=45, twistLimitDegrees=5}
 {bone1Name=dyn_coat_r_02, bone2Name=dyn_coat_r_01, swing1LimitDegrees=10, swing2LimitDegrees=45, twistLimitDegrees=5}
 {bone1Name=dyn_coat_r_03, bone2Name=dyn_coat_r_02, swing1LimitDegrees=10, swing2LimitDegrees=45, twistLimitDegrees=5}
 {
  bone1Name=dyn_coat_l_MSTR
  bone2Name=spine_03
  swing1LimitDegrees=45
  swing1Motion=Free
  swing2LimitDegrees=45
  swing2Motion=Free
  twistLimitDegrees=45
  twistMotion=Free
 }
 {
  bone1Name=dyn_coat_r_MSTR
  bone2Name=spine_03
  swing1LimitDegrees=45
  swing1Motion=Free
  swing2LimitDegrees=45
  swing2Motion=Free
  twistLimitDegrees=45
  twistMotion=Free
 }
 {
  bone1Name=thigh_l
  bone2Name=pelvis
  swing1LimitDegrees=45
  swing1Motion=Free
  swing2LimitDegrees=45
  swing2Motion=Free
  twistLimitDegrees=45
  twistMotion=Free
 }
 {
  bone1Name=spine_01
  bone2Name=pelvis
  swing1LimitDegrees=45
  swing1Motion=Free
  swing2LimitDegrees=45
  swing2Motion=Free
  twistLimitDegrees=45
  twistMotion=Free
 }
 {
  bone1Name=spine_02
  bone2Name=spine_01
  swing1LimitDegrees=45
  swing1Motion=Free
  swing2LimitDegrees=45
  swing2Motion=Free
  twistLimitDegrees=45
  twistMotion=Free
 }
 {
  bone1Name=thigh_r
  bone2Name=pelvis
  swing1LimitDegrees=45
  swing1Motion=Free
  swing2LimitDegrees=45
  swing2Motion=Free
  twistLimitDegrees=45
  twistMotion=Free
 }
 {
  bone1Name=spine_03
  bone2Name=spine_02
  swing1LimitDegrees=45
  swing1Motion=Free
  swing2LimitDegrees=45
  swing2Motion=Free
  twistLimitDegrees=45
  twistMotion=Free
 }
]
