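# Switch AnimBP for the M_MED_Gimmick character body, animating coat, belts, lapels and wrist straps.
# Parent-mesh pose passes through deformation Control Rigs, strap constraints and AnimDynamics into a shared cache.
# Output selects RigidBody coat and belt simulation when enabled, otherwise Trail controllers, with a re-posed DBNO variant.
# The event graph computes RigidBody gravity, movement-based trail alpha, cloth wind gusts, and a skydive/swim trail rotation the output never uses.
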AnimBlueprint /BRCosmetics/Characters/Player/Male/Medium/Bodies/M_MED_Gimmick/Meshes/ClothAsset/ClothForSwitch/M_MED_Gimmick_AnimBP_ForSwitch
target_skeleton=/BRCosmetics/Characters/Player/Male/Medium/Bodies/M_MED_Gimmick/Meshes/M_MED_Gimmick_Skeleton
graphs [
 AnimGraph {nodes [
   n1 = "Transform (Modify) Bone - Bone: dyn_coat_l_03"(Alpha=1, Translation="0,0,0", Rotation="0,0,0", Scale="1,1,1") {rotationMode=BMM_Additive, rotationSpace=BCS_BoneSpace, alphaBoolBlend {blendOption=Linear}, boneToModify {boneName=dyn_coat_l_03}}
   n2 = "Transform (Modify) Bone - Bone: dyn_coat_bk_01"(ComponentPose=@n1, Alpha=1, Translation="0,0,0", Rotation="0,0,0", Scale="1,1,1") {rotationMode=BMM_Additive, rotationSpace=BCS_BoneSpace, alphaBoolBlend {blendOption=Linear}, boneToModify {boneName=dyn_coat_bk_01}}
   "Transform (Modify) Bone - Bone: dyn_coat_r_03"(ComponentPose=@n2, Alpha=1, Translation="0,0,0", Rotation="0,0,0", Scale="1,1,1") {rotationMode=BMM_Additive, rotationSpace=BCS_BoneSpace, alphaBoolBlend {blendOption=Linear}, boneToModify {boneName=dyn_coat_r_03}}
   n3 = Use cached pose 'trailCache'()
   n4 = Use cached pose 'dbnoCache'()
   n5 = Get bIsDBNO()
   Output Pose(Result=Blend Poses by bool(
     BlendPose_0=Use cached pose 'RbanCache'()
     BlendPose_1=Blend Poses by bool(BlendPose_0=@n4, BlendPose_1=@n3, bActiveValue=@n5) {blendType=Linear}
     bActiveValue=Get bRigidBodyEnabled()
    ) {blendType=Linear}) {layerGroup=None}
   n6 = Control Rig(Source=Copy Pose From Mesh() {bUseAttachedParent=true}) {
    alphaBoolBlend {blendOption=Linear}
    controlRigAssetReference {blueprintRigClass=/BRCosmetics/Characters/Player/Male/Medium/Bodies/M_MED_Gimmick/Meshes/M_MED_Gimmick_Deform_CtrlRig_twists.M_MED_Gimmick_Deform_CtrlRig_twists_C}
   }
   n7 = Control Rig!(
    Source=@n6
    Alpha
    KneeRot_fr_l_114deg:Vector
    KneeTrans_fr_l_114deg:Vector="0.258,-3.591,0.009"
    KneeScale_fr_l_114deg:Vector="1,1,1"
    KneeScale_fr_l_130deg:Vector="1,1,1"
    KneeRot_fr_l_130deg:Vector="0,0,-51.2"
    KneeTrans_fr_l_130deg:Vector="0.258,-3.591,0.009"
   ) -> (Pose) {
    alphaBoolBlend {blendOption=Linear}
    controlRigAssetReference {blueprintRigClass=/BRCosmetics/Characters/Player/Male/Medium/Bodies/M_MED_Gimmick/Meshes/M_MED_Gimmick_Deform_CtrlRig_sculpters.M_MED_Gimmick_Deform_CtrlRig_sculpters_C}
   }
   n8 = "Constraint - dyn_wrist_strap_l_01_MSTR (PPP)"!(
    ComponentPose=Local To Component(LocalPose=@n7)
    Alpha=1
    ConstraintWeights_0=0.4869
    ConstraintWeights_1:float=0.3816
    ConstraintWeights_2:float=0.1316
   ) -> (Pose) {
    alphaBoolBlend {blendOption=Linear}
    boneToModify {boneName=dyn_wrist_strap_l_01_MSTR}
    constraintSetup [{targetBone {boneName=lowerarm_twist_02_l}}, {targetBone {boneName=lowerarm_twist_01_l}}, {targetBone {boneName=lowerarm_l}}]
    constraintWeights [1,1,1]
   }
   n9 = "Constraint - dyn_wrist_strap_l_02_MSTR (PPPPP)"!(
    ComponentPose=@n8
    Alpha=1
    ConstraintWeights_0=0.5451
    ConstraintWeights_1:float=0.3057
    ConstraintWeights_2:float=0.0664
    ConstraintWeights_3:float=0.0455
    ConstraintWeights_4:float=0.0373
   ) -> (Pose) {
    alphaBoolBlend {blendOption=Linear}
    boneToModify {boneName=dyn_wrist_strap_l_02_MSTR}
    constraintSetup [
     {targetBone {boneName=lowerarm_twist_01_l}}
     {targetBone {boneName=lowerarm_twist_02_l}}
     {targetBone {boneName=lowerarm_l}}
     {targetBone {boneName=pinky_metacarpal_l}}
     {targetBone {boneName=hand_l}}
    ]
    constraintWeights [1,1,1,1,1]
   }
   n10 = "Constraint - dyn_wrist_strap_r_01_MSTR (PPP)"!(ComponentPose=@n9, Alpha=1, ConstraintWeights_0=0.4869, ConstraintWeights_1:float=0.3816, ConstraintWeights_2:float=0.1316) -> (Pose) {
    alphaBoolBlend {blendOption=Linear}
    boneToModify {boneName=dyn_wrist_strap_r_01_MSTR}
    constraintSetup [{targetBone {boneName=lowerarm_twist_02_r}}, {targetBone {boneName=lowerarm_twist_01_r}}, {targetBone {boneName=lowerarm_r}}]
    constraintWeights [1,1,1]
   }
   n11 = "Constraint - dyn_wrist_strap_r_02_MSTR (PPPPP)"!(
    ComponentPose=@n10
    Alpha=1
    ConstraintWeights_0=0.5451
    ConstraintWeights_1:float=0.3057
    ConstraintWeights_2:float=0.0664
    ConstraintWeights_3:float=0.0455
    ConstraintWeights_4:float=0.0373
   ) -> (Pose) {
    alphaBoolBlend {blendOption=Linear}
    boneToModify {boneName=dyn_wrist_strap_r_02_MSTR}
    constraintSetup [
     {targetBone {boneName=lowerarm_twist_01_r}}
     {targetBone {boneName=lowerarm_twist_02_r}}
     {targetBone {boneName=lowerarm_r}}
     {targetBone {boneName=pinky_metacarpal_r}}
     {targetBone {boneName=hand_r}}
    ]
    constraintWeights [1,1,1,1,1]
   }
   n12 = Anim Dynamics - Bone: dyn_lapel_l(ComponentPose=@n11, Alpha=1, ExternalForce="0,0,0") {
    angularDampingOverride=1
    angularSpringConstant=20000
    bAngularSpring=true
    bOverrideAngularDamping=true
    bUseGravityOverride=true
    gravityScale=0.35
    simulationSpace=RootRelative
    alphaBoolBlend {blendOption=Linear}
    boundBone {boneName=dyn_lapel_l}
    componentAppliedLinearAccClamp=<3000 3000 3000>
    componentLinearAccScale=<1 1 1>
    componentLinearVelScale=<1 1 1>
    gravityOverride {z=-980}
    physicsBodyDefinitions [{
      sphereCollisionRadius=0
      boundBone {boneName=dyn_lapel_l}
      boxExtents=<20 30 30>
      constraintSetup {angularLimitsMax=<0 0 0>, angularLimitsMin=<0 -30 0>, angularTarget=<1 0 0>}
      localJointOffset=<4 0 0>
     }]
   }
   n13 = Anim Dynamics - Bone: dyn_rapel_r(ComponentPose=@n12, Alpha=1, ExternalForce="0,0,0") {
    angularDampingOverride=1
    angularSpringConstant=20000
    bAngularSpring=true
    bOverrideAngularDamping=true
    bUseGravityOverride=true
    gravityScale=0.35
    simulationSpace=RootRelative
    alphaBoolBlend {blendOption=Linear}
    boundBone {boneName=dyn_rapel_r}
    componentAppliedLinearAccClamp=<3000 3000 3000>
    componentLinearAccScale=<1 1 1>
    componentLinearVelScale=<1 1 1>
    gravityOverride {z=-980}
    physicsBodyDefinitions [{
      sphereCollisionRadius=0
      boundBone {boneName=dyn_rapel_r}
      boxExtents=<20 30 30>
      constraintSetup {angularLimitsMax=<0 0 0>, angularLimitsMin=<0 -30 0>, angularTarget=<1 0 0>}
      localJointOffset=<4 0 0>
     }]
   }
   n14 = Anim Dynamics - Bone: dyn_lapel_high_l(ComponentPose=@n13, Alpha=1, ExternalForce="0,0,0") {
    angularDampingOverride=1
    angularSpringConstant=20000
    bAngularSpring=true
    bOverrideAngularDamping=true
    bUseGravityOverride=true
    gravityScale=0.35
    simulationSpace=RootRelative
    alphaBoolBlend {blendOption=Linear}
    boundBone {boneName=dyn_lapel_high_l}
    componentAppliedLinearAccClamp=<3000 3000 3000>
    componentLinearAccScale=<1 1 1>
    componentLinearVelScale=<1 1 1>
    gravityOverride {z=-980}
    physicsBodyDefinitions [{
      sphereCollisionRadius=0
      boundBone {boneName=dyn_lapel_high_l}
      boxExtents=<20 30 30>
      constraintSetup {angularLimitsMax=<0 0 0>, angularLimitsMin=<0 -30 0>, angularTarget=<1 0 0>}
      localJointOffset=<4 0 0>
     }]
   }
   n15 = Anim Dynamics - Bone: dyn_rapel_high_r(ComponentPose=@n14, Alpha=1, ExternalForce="0,0,0") {
    angularDampingOverride=1
    angularSpringConstant=20000
    bAngularSpring=true
    bOverrideAngularDamping=true
    bUseGravityOverride=true
    gravityScale=0.35
    simulationSpace=RootRelative
    alphaBoolBlend {blendOption=Linear}
    boundBone {boneName=dyn_rapel_high_r}
    componentAppliedLinearAccClamp=<3000 3000 3000>
    componentLinearAccScale=<1 1 1>
    componentLinearVelScale=<1 1 1>
    gravityOverride {z=-980}
    physicsBodyDefinitions [{
      sphereCollisionRadius=0
      boundBone {boneName=dyn_rapel_high_r}
      boxExtents=<20 30 30>
      constraintSetup {angularLimitsMax=<0 0 0>, angularLimitsMin=<0 -30 0>, angularTarget=<1 0 0>}
      localJointOffset=<4 0 0>
     }]
   }
   n16 = Anim Dynamics - Bone: dyn_lapel_low_l(ComponentPose=@n15, Alpha=1, ExternalForce="0,0,0") {
    angularDampingOverride=1
    angularSpringConstant=20000
    bAngularSpring=true
    bOverrideAngularDamping=true
    bUseGravityOverride=true
    gravityScale=0.35
    simulationSpace=RootRelative
    alphaBoolBlend {blendOption=Linear}
    boundBone {boneName=dyn_lapel_low_l}
    componentAppliedLinearAccClamp=<3000 3000 3000>
    componentLinearAccScale=<1 1 1>
    componentLinearVelScale=<1 1 1>
    gravityOverride {z=-980}
    physicsBodyDefinitions [{
      sphereCollisionRadius=0
      boundBone {boneName=dyn_lapel_low_l}
      boxExtents=<20 30 30>
      constraintSetup {angularLimitsMax=<0 0 0>, angularLimitsMin=<0 -30 0>, angularTarget=<1 0 0>}
      localJointOffset=<4 0 0>
     }]
   }
   n17 = Anim Dynamics - Bone: dyn_rapel_row_r(ComponentPose=@n16, Alpha=1, ExternalForce="0,0,0") {
    angularDampingOverride=1
    angularSpringConstant=20000
    bAngularSpring=true
    bOverrideAngularDamping=true
    bUseGravityOverride=true
    gravityScale=0.35
    simulationSpace=RootRelative
    alphaBoolBlend {blendOption=Linear}
    boundBone {boneName=dyn_rapel_row_r}
    componentAppliedLinearAccClamp=<3000 3000 3000>
    componentLinearAccScale=<1 1 1>
    componentLinearVelScale=<1 1 1>
    gravityOverride {z=-980}
    physicsBodyDefinitions [{
      sphereCollisionRadius=0
      boundBone {boneName=dyn_rapel_row_r}
      boxExtents=<20 30 30>
      constraintSetup {angularLimitsMax=<0 0 0>, angularLimitsMin=<0 -30 0>, angularTarget=<1 0 0>}
      localJointOffset=<4 0 0>
     }]
   }
   n18 = Anim Dynamics - Bone: dyn_wrist_strap_l_01(ComponentPose=@n17, Alpha=1, ExternalForce="0,0,0") {
    angularDampingOverride=1
    angularSpringConstant=20000
    bAngularSpring=true
    bOverrideAngularDamping=true
    bUseGravityOverride=true
    gravityScale=0.35
    simulationSpace=RootRelative
    alphaBoolBlend {blendOption=Linear}
    boundBone {boneName=dyn_wrist_strap_l_01}
    componentAppliedLinearAccClamp=<3000 3000 3000>
    componentLinearAccScale=<1 1 1>
    componentLinearVelScale=<1 1 1>
    gravityOverride {z=-980}
    physicsBodyDefinitions [{
      sphereCollisionRadius=0
      boundBone {boneName=dyn_wrist_strap_l_01}
      boxExtents=<20 30 30>
      constraintSetup {angularLimitsMax=<0 15 0>, angularLimitsMin=<0 -15 0>, angularTarget=<1 0 0>}
      localJointOffset=<4 0 0>
     }]
   }
   n19 = Anim Dynamics - Bone: dyn_wrist_strap_l_02(ComponentPose=@n18, Alpha=1, ExternalForce="0,0,0") {
    angularDampingOverride=1
    angularSpringConstant=20000
    bAngularSpring=true
    bOverrideAngularDamping=true
    bUseGravityOverride=true
    gravityScale=0.35
    simulationSpace=RootRelative
    alphaBoolBlend {blendOption=Linear}
    boundBone {boneName=dyn_wrist_strap_l_02}
    componentAppliedLinearAccClamp=<3000 3000 3000>
    componentLinearAccScale=<1 1 1>
    componentLinearVelScale=<1 1 1>
    gravityOverride {z=-980}
    physicsBodyDefinitions [{
      sphereCollisionRadius=0
      boundBone {boneName=dyn_wrist_strap_l_02}
      boxExtents=<20 30 30>
      constraintSetup {angularLimitsMax=<0 15 0>, angularLimitsMin=<0 -15 0>, angularTarget=<1 0 0>}
      localJointOffset=<4 0 0>
     }]
   }
   n20 = Anim Dynamics - Bone: dyn_wrist_strap_r_01(ComponentPose=@n19, Alpha=1, ExternalForce="0,0,0") {
    angularDampingOverride=1
    angularSpringConstant=20000
    bAngularSpring=true
    bOverrideAngularDamping=true
    bUseGravityOverride=true
    gravityScale=0.35
    simulationSpace=RootRelative
    alphaBoolBlend {blendOption=Linear}
    boundBone {boneName=dyn_wrist_strap_r_01}
    componentAppliedLinearAccClamp=<3000 3000 3000>
    componentLinearAccScale=<1 1 1>
    componentLinearVelScale=<1 1 1>
    gravityOverride {z=-980}
    physicsBodyDefinitions [{
      sphereCollisionRadius=0
      boundBone {boneName=dyn_wrist_strap_r_01}
      boxExtents=<20 30 30>
      constraintSetup {angularLimitsMax=<0 15 0>, angularLimitsMin=<0 -15 0>, angularTarget=<1 0 0>}
      localJointOffset=<4 0 0>
     }]
   }
   n21 = Anim Dynamics - Bone: dyn_wrist_strap_r_02(ComponentPose=@n20, Alpha=1, ExternalForce="0,0,0") {
    angularDampingOverride=1
    angularSpringConstant=20000
    bAngularSpring=true
    bOverrideAngularDamping=true
    bUseGravityOverride=true
    gravityScale=0.35
    simulationSpace=RootRelative
    alphaBoolBlend {blendOption=Linear}
    boundBone {boneName=dyn_wrist_strap_r_02}
    componentAppliedLinearAccClamp=<3000 3000 3000>
    componentLinearAccScale=<1 1 1>
    componentLinearVelScale=<1 1 1>
    gravityOverride {z=-980}
    physicsBodyDefinitions [{
      sphereCollisionRadius=0
      boundBone {boneName=dyn_wrist_strap_r_02}
      boxExtents=<20 30 30>
      constraintSetup {angularLimitsMax=<0 15 0>, angularLimitsMin=<0 -15 0>, angularTarget=<1 0 0>}
      localJointOffset=<4 0 0>
     }]
   }
   n22 = Control Rig!(Source=Component To Local(ComponentPose=@n21), Alpha) -> (Pose) {
    alphaBoolBlend {blendOption=Linear}
    controlRigAssetReference {blueprintRigClass=/BRCosmetics/Characters/Player/Male/Medium/Bodies/M_MED_Gimmick/Meshes/M_MED_Gimmick_CtrlRig.M_MED_Gimmick_CtrlRig_C}
   }
   Save cached pose 'StartupCache'(Pose=@n22) {cacheName=StartupCache}
   n23 = Get BwdMovement() -> (BwdMovement:double)
   n24 = Trail controller - Bone: dyn_coat_l_06(ComponentPose=Local To Component(LocalPose=Use cached pose 'StartupCache'()), Alpha=@n23) {
    bUsePlanarLimit=true
    chainLength=4
    debugLifeTime=0.1
    maxDeltaTime=0.333
    alphaScaleBiasClamp {bInterpResult=true}
    rotationLimits [{limitMax=<180 180 180>}, {limitMax=<180 180 180>}, {limitMax=<180 180 180>}, {limitMax=<180 180 180>}]
    rotationOffsets [<0 0 0>, <0 0 0>, <0 0 0>, <0 0 0>]
    trailBone {boneName=dyn_coat_l_06}
    trailRelaxationSpeed {editorCurveData {keys [<at -0.0134 19.3841 RCIM_Linear>,<at 0.5079 12.8986 RCIM_Linear>,<at 0.9955 9.6044 RCIM_Linear>]}}
   }
   n25 = Trail controller - Bone: dyn_coat_bk_l_06(ComponentPose=@n24, Alpha=@n23) {
    bUsePlanarLimit=true
    chainLength=4
    debugLifeTime=0.1
    maxDeltaTime=0.333
    alphaScaleBiasClamp {bInterpResult=true}
    rotationLimits [{limitMax=<180 180 180>}, {limitMax=<180 180 180>}, {limitMax=<180 180 180>}, {limitMax=<180 180 180>}]
    rotationOffsets [<0 0 0>, <0 0 0>, <0 0 0>, <0 0 0>]
    trailBone {boneName=dyn_coat_bk_l_06}
    trailRelaxationSpeed {editorCurveData {keys [<at -0.0134 19.3841 RCIM_Linear>,<at 0.5079 12.8986 RCIM_Linear>,<at 0.9955 9.6044 RCIM_Linear>]}}
   }
   n26 = Trail controller - Bone: dyn_coat_bk_04(ComponentPose=@n25, Alpha=@n23) {
    bUsePlanarLimit=true
    debugLifeTime=0.1
    maxDeltaTime=0.0333
    alphaScaleBiasClamp {bInterpResult=true}
    rotationLimits [{limitMax=<180 180 180>}, {limitMax=<180 180 180>}]
    rotationOffsets [<0 0 0>, <0 0 0>]
    trailBone {boneName=dyn_coat_bk_04}
    trailRelaxationSpeed {editorCurveData {keys [<at 0 15 RCIM_Linear>,<at 0.5289 13.1838 RCIM_Linear>,<at 1 6 RCIM_Linear>]}}
   }
   n27 = Trail controller - Bone: dyn_coat_bk_r_06(ComponentPose=@n26, Alpha=@n23) {
    bUsePlanarLimit=true
    chainLength=4
    debugLifeTime=0.1
    maxDeltaTime=0.333
    alphaScaleBiasClamp {bInterpResult=true}
    rotationLimits [{limitMax=<180 180 180>}, {limitMax=<180 180 180>}, {limitMax=<180 180 180>}, {limitMax=<180 180 180>}]
    rotationOffsets [<0 0 0>, <0 0 0>, <0 0 0>, <0 0 0>]
    trailBone {boneName=dyn_coat_bk_r_06}
    trailRelaxationSpeed {editorCurveData {keys [<at -0.0134 19.3841 RCIM_Linear>,<at 0.5079 12.8986 RCIM_Linear>,<at 0.9955 9.6044 RCIM_Linear>]}}
   }
   n28 = Trail controller - Bone: dyn_coat_r_06(ComponentPose=@n27, Alpha=@n23) {
    bUsePlanarLimit=true
    chainLength=4
    debugLifeTime=0.1
    maxDeltaTime=0.0333
    alphaScaleBiasClamp {bInterpResult=true}
    rotationLimits [{limitMax=<180 180 180>}, {limitMax=<180 180 180>}, {limitMax=<180 90 180>, limitMin=<-180 -90 -180>}, {limitMax=<180 180 180>}]
    rotationOffsets [<0 0 0>, <0 0 0>, <0 0 0>, <0 0 0>]
    trailBone {boneName=dyn_coat_r_06}
    trailRelaxationSpeed {editorCurveData {keys [<at 0 15 RCIM_Linear>,<at 0.5289 13.1838 RCIM_Linear>,<at 1 6 RCIM_Linear>]}}
   }
   n29 = "Constraint - dyn_belt_r_MSTR (PPPPPPPP)"!(
    ComponentPose=@n28
    Alpha=1
    ConstraintWeights_0=0.241
    ConstraintWeights_1:float=0.2059
    ConstraintWeights_2:float=0.1856
    ConstraintWeights_3:float=0.1383
    ConstraintWeights_4:float=0.097
    ConstraintWeights_5:float=0.0649
    ConstraintWeights_6:float=0.0375
    ConstraintWeights_7:float=0.0297
   ) -> (Pose) {
    alphaBoolBlend {blendOption=Linear}
    boneToModify {boneName=dyn_belt_r_MSTR}
    constraintSetup [
     {targetBone {boneName=pelvis}}
     {targetBone {boneName=dyn_coat_r_03}}
     {targetBone {boneName=spine_01}}
     {targetBone {boneName=dyn_coat_r_02}}
     {targetBone {boneName=spine_02}}
     {targetBone {boneName=dyn_coat_r_01}}
     {targetBone {boneName=spine_03}}
     {targetBone {boneName=dyn_coat_bk_r_02}}
    ]
    constraintWeights [1,1,1,1,1,1,1,1]
   }
   n30 = "Constraint - dyn_belt_l_MSTR (PPPPPPP)"!(
    ComponentPose=@n29
    Alpha=1
    ConstraintWeights_0=0.2394
    ConstraintWeights_1:float=0.2063
    ConstraintWeights_2:float=0.1922
    ConstraintWeights_3:float=0.1402
    ConstraintWeights_4:float=0.1043
    ConstraintWeights_5:float=0.07
    ConstraintWeights_6:float=0.0475
   ) -> (Pose) {
    alphaBoolBlend {blendOption=Linear}
    boneToModify {boneName=dyn_belt_l_MSTR}
    constraintSetup [
     {targetBone {boneName=pelvis}}
     {targetBone {boneName=dyn_coat_l_03}}
     {targetBone {boneName=spine_01}}
     {targetBone {boneName=dyn_coat_l_02}}
     {targetBone {boneName=spine_02}}
     {targetBone {boneName=dyn_coat_l_01}}
     {targetBone {boneName=spine_03}}
    ]
    constraintWeights [1,1,1,1,1,1,1]
   }
   Save cached pose 'trailCache'(Pose=Component To Local(ComponentPose=@n30)) {cacheName=trailCache}
   n31 = "Transform (Modify) Bone - Bone: dyn_coat_l_01"(
    ComponentPose=Local To Component(LocalPose=Use cached pose 'StartupCache'())
    Alpha=1
    Translation="0,0,0"
    Rotation="32.145,11.529,-3.902"
    Scale="1,1,1"
   ) {
    rotationMode=BMM_Additive
    rotationSpace=BCS_BoneSpace
    alphaBoolBlend {blendOption=Linear}
    boneToModify {boneName=dyn_coat_l_01}
    rotation=<rot 32.145 11.529 -3.902>
   }
   n32 = "Transform (Modify) Bone - Bone: dyn_coat_l_03"(ComponentPose=@n31, Alpha=1, Translation="0,0,0", Rotation="-0.909,17.433,8.209", Scale="1,1,1") {
    rotationMode=BMM_Additive
    rotationSpace=BCS_BoneSpace
    alphaBoolBlend {blendOption=Linear}
    boneToModify {boneName=dyn_coat_l_03}
    rotation=<rot -0.909 17.433 8.209>
   }
   n33 = "Transform (Modify) Bone - Bone: dyn_coat_l_05"(ComponentPose=@n32, Alpha=1, Translation="0,0,0", Rotation="5.741,-9.962,0.410", Scale="1,1,1") {
    rotationMode=BMM_Additive
    rotationSpace=BCS_BoneSpace
    alphaBoolBlend {blendOption=Linear}
    boneToModify {boneName=dyn_coat_l_05}
    rotation=<rot 5.741 -9.9619 0.41>
   }
   n34 = "Transform (Modify) Bone - Bone: dyn_coat_bk_l_02"(ComponentPose=@n33, Alpha=1, Translation="0,0,0", Rotation="-8.860,3.786,-1.333", Scale="1,1,1") {
    rotationMode=BMM_Additive
    rotationSpace=BCS_BoneSpace
    alphaBoolBlend {blendOption=Linear}
    boneToModify {boneName=dyn_coat_bk_l_02}
    rotation=<rot -8.86 3.786 -1.333>
   }
   n35 = "Transform (Modify) Bone - Bone: dyn_coat_bk_l_04"(ComponentPose=@n34, Alpha=1, Translation="0,0,0", Rotation="-0.703,3.196,7.134", Scale="1,1,1") {
    rotationMode=BMM_Additive
    rotationSpace=BCS_BoneSpace
    alphaBoolBlend {blendOption=Linear}
    boneToModify {boneName=dyn_coat_bk_l_04}
    rotation=<rot -0.703 3.196 7.134>
   }
   n36 = "Transform (Modify) Bone - Bone: dyn_coat_bk_l_05"(ComponentPose=@n35, Alpha=1, Translation="0,0,0", Rotation="-15.765,19.004,-5.367", Scale="1,1,1") {
    rotationMode=BMM_Additive
    rotationSpace=BCS_BoneSpace
    alphaBoolBlend {blendOption=Linear}
    boneToModify {boneName=dyn_coat_bk_l_05}
    rotation=<rot -15.765 19.004 -5.367>
   }
   n37 = "Transform (Modify) Bone - Bone: dyn_coat_bk_01"(ComponentPose=@n36, Alpha=1, Translation="0,0,0", Rotation="-15.781,-0.215,-1.367", Scale="1,1,1") {
    rotationMode=BMM_Additive
    rotationSpace=BCS_BoneSpace
    alphaBoolBlend {blendOption=Linear}
    boneToModify {boneName=dyn_coat_bk_01}
    rotation=<rot -15.781 -0.215 -1.367>
   }
   n38 = "Transform (Modify) Bone - Bone: dyn_coat_bk_02"(ComponentPose=@n37, Alpha=1, Translation="0,0,0", Rotation="-7.246,0.129,0.127", Scale="1,1,1") {
    rotationMode=BMM_Additive
    rotationSpace=BCS_BoneSpace
    alphaBoolBlend {blendOption=Linear}
    boneToModify {boneName=dyn_coat_bk_02}
    rotation=<rot -7.246 0.129 0.127>
   }
   n39 = "Transform (Modify) Bone - Bone: dyn_coat_bk_04"(ComponentPose=@n38, Alpha=1, Translation="0,0,0", Rotation="-10.235,-0.453,0.706", Scale="1,1,1") {
    rotationMode=BMM_Additive
    rotationSpace=BCS_BoneSpace
    alphaBoolBlend {blendOption=Linear}
    boneToModify {boneName=dyn_coat_bk_04}
    rotation=<rot -10.235 -0.453 0.706>
   }
   n40 = "Transform (Modify) Bone - Bone: dyn_coat_bk_split_l_01"(ComponentPose=@n39, Alpha=1, Translation="0,0,0", Rotation="-2.311,-0.443,0.100", Scale="1,1,1") {
    rotationMode=BMM_Additive
    rotationSpace=BCS_BoneSpace
    alphaBoolBlend {blendOption=Linear}
    boneToModify {boneName=dyn_coat_bk_split_l_01}
    rotation=<rot -2.311 -0.443 0.1>
   }
   n41 = "Transform (Modify) Bone - Bone: dyn_coat_bk_split_r_01"(ComponentPose=@n40, Alpha=1, Translation="0,0,0", Rotation="-1.460,-6.056,1.270", Scale="1,1,1") {
    rotationMode=BMM_Additive
    rotationSpace=BCS_BoneSpace
    alphaBoolBlend {blendOption=Linear}
    boneToModify {boneName=dyn_coat_bk_split_r_01}
    rotation=<rot -1.46 -6.056 1.27>
   }
   n42 = "Transform (Modify) Bone - Bone: dyn_coat_bk_r_02"(ComponentPose=@n41, Alpha=1, Translation="0,0,0", Rotation="-10.873,-8.896,1.662", Scale="1,1,1") {
    rotationMode=BMM_Additive
    rotationSpace=BCS_BoneSpace
    alphaBoolBlend {blendOption=Linear}
    boneToModify {boneName=dyn_coat_bk_r_02}
    rotation=<rot -10.873 -8.896 1.662>
   }
   n43 = "Transform (Modify) Bone - Bone: dyn_coat_bk_r_04"(ComponentPose=@n42, Alpha=1, Translation="0,0,0", Rotation="-7.250,-3.668,-0.463", Scale="1,1,1") {
    rotationMode=BMM_Additive
    rotationSpace=BCS_BoneSpace
    alphaBoolBlend {blendOption=Linear}
    boneToModify {boneName=dyn_coat_bk_r_04}
    rotation=<rot -7.25 -3.668 -0.463>
   }
   n44 = "Transform (Modify) Bone - Bone: dyn_coat_bk_r_05"(ComponentPose=@n43, Alpha=1, Translation="0,0,0", Rotation="-19.163,-11.226,1.333", Scale="1,1,1") {
    rotationMode=BMM_Additive
    rotationSpace=BCS_BoneSpace
    alphaBoolBlend {blendOption=Linear}
    boneToModify {boneName=dyn_coat_bk_r_05}
    rotation=<rot -19.163 -11.226 1.333>
   }
   n45 = "Transform (Modify) Bone - Bone: dyn_coat_r_01"(ComponentPose=@n44, Alpha=1, Translation="0,0,0", Rotation="15.009,1.233,0.350", Scale="1,1,1") {
    rotationMode=BMM_Additive
    rotationSpace=BCS_BoneSpace
    alphaBoolBlend {blendOption=Linear}
    boneToModify {boneName=dyn_coat_r_01}
    rotation=<rot 15.009 1.233 0.35>
   }
   n46 = "Transform (Modify) Bone - Bone: dyn_coat_r_03"(ComponentPose=@n45, Alpha=1, Translation="0,0,0", Rotation="-21.159,-40.049,0.078", Scale="1,1,1") {
    rotationMode=BMM_Additive
    rotationSpace=BCS_BoneSpace
    alphaBoolBlend {blendOption=Linear}
    boneToModify {boneName=dyn_coat_r_03}
    rotation=<rot -21.159 -40.049 0.078>
   }
   n47 = "Transform (Modify) Bone - Bone: dyn_coat_r_05"(ComponentPose=@n46, Alpha=1, Translation="0,0,0", Rotation="0,10,0", Scale="1,1,1") {
    rotationMode=BMM_Additive
    rotationSpace=BCS_BoneSpace
    alphaBoolBlend {blendOption=Linear}
    boneToModify {boneName=dyn_coat_r_05}
    rotation {yaw=10}
   }
   n48 = Look At - Bone: dyn_coat_l_01(ComponentPose=@n47, Alpha=1) {
    bUseLookUpAxis=true
    alphaBoolBlend {blendOption=Linear}
    boneToModify {boneName=dyn_coat_l_01}
    lookAtLocation=<0.0206 -1.6234 -5.8436>
    lookAtTarget {boneReference {boneName=foot_l}}
    lookAt_Axis {axis=<1 0>}
    lookUp_Axis {axis {x=0, z=1}}
   }
   n49 = Look At - Bone: dyn_coat_r_01(ComponentPose=@n48, Alpha=1) {
    bUseLookUpAxis=true
    alphaBoolBlend {blendOption=Linear}
    boneToModify {boneName=dyn_coat_r_01}
    lookAtLocation=<7.3292 23.916 9.591>
    lookAtTarget {boneReference {boneName=foot_r}}
    lookAt_Axis {axis=<1 0>}
    lookUp_Axis {axis {x=0, z=1}}
   }
   n50 = Trail controller - Bone: dyn_coat_l_06!(ComponentPose=@n49, Alpha=1, RelaxationSpeedScale:float=2, FakeVelocity:Vector="0,0,0") -> (Pose) {
    bShowTrailLocation=true
    bUsePlanarLimit=true
    chainLength=3
    debugLifeTime=0.1
    lastBoneRotationAnimAlphaBlend=1
    maxDeltaTime=0.0333
    relaxationSpeedScale=1.2
    alphaScaleBiasClamp {bInterpResult=true, interpSpeedDecreasing=25, interpSpeedIncreasing=25}
    rotationLimits [{limitMax=<180 180 180>}, {limitMax=<180 180 180>}, {limitMax=<180 180 180>}]
    rotationOffsets [<0 0 0>, <0 0 0>, <0 0 0>]
    trailBone {boneName=dyn_coat_l_06}
    trailRelaxationSpeed {editorCurveData {keys [<at 0 15 RCIM_Linear>,<at 1 6 RCIM_Linear>]}}
   }
   n51 = Trail controller - Bone: dyn_coat_bk_l_06!(ComponentPose=@n50, Alpha=1, RelaxationSpeedScale:float=2, FakeVelocity:Vector="0,0,0") -> (Pose) {
    bShowTrailLocation=true
    bUsePlanarLimit=true
    chainLength=4
    debugLifeTime=0.1
    lastBoneRotationAnimAlphaBlend=1
    maxDeltaTime=0.0333
    relaxationSpeedScale=1.2
    alphaScaleBiasClamp {bInterpResult=true, interpSpeedDecreasing=25, interpSpeedIncreasing=25}
    rotationLimits [{limitMax=<180 180 180>}, {limitMax=<180 180 180>}, {limitMax=<180 180 180>}, {limitMax=<180 180 180>}]
    rotationOffsets [<0 0 0>, <0 0 0>, <0 0 0>, <0 0 0>]
    trailBone {boneName=dyn_coat_bk_l_06}
    trailRelaxationSpeed {editorCurveData {keys [<at 0 15 RCIM_Linear>,<at 1 6 RCIM_Linear>]}}
   }
   n52 = Trail controller - Bone: dyn_coat_bk_04!(ComponentPose=@n51, Alpha=1, RelaxationSpeedScale:float=2, FakeVelocity:Vector="0,0,0") -> (Pose) {
    bReorientParentToChild=false
    bShowTrailLocation=true
    bUsePlanarLimit=true
    debugLifeTime=0.1
    lastBoneRotationAnimAlphaBlend=1
    maxDeltaTime=0.0333
    relaxationSpeedScale=1.2
    alphaScaleBiasClamp {bInterpResult=true, interpSpeedDecreasing=25, interpSpeedIncreasing=25}
    rotationLimits [{limitMax=<180 180 180>}, {limitMax=<180 180 180>}]
    rotationOffsets [<0 0 0>, <0 0 0>]
    trailBone {boneName=dyn_coat_bk_04}
    trailRelaxationSpeed {editorCurveData {keys [<at 0 15 RCIM_Linear>,<at 1 6 RCIM_Linear>]}}
   }
   n53 = Trail controller - Bone: dyn_coat_bk_r_06!(ComponentPose=@n52, Alpha=1, RelaxationSpeedScale:float=2, FakeVelocity:Vector="0,0,0") -> (Pose) {
    bShowTrailLocation=true
    bUsePlanarLimit=true
    chainLength=3
    debugLifeTime=0.1
    lastBoneRotationAnimAlphaBlend=1
    maxDeltaTime=0.0333
    relaxationSpeedScale=1.2
    alphaScaleBiasClamp {bInterpResult=true, interpSpeedDecreasing=25, interpSpeedIncreasing=25}
    rotationLimits [{limitMax=<180 180 180>}, {limitMax=<180 180 180>}, {limitMax=<180 180 180>}]
    rotationOffsets [<0 0 0>, <0 0 0>, <0 0 0>]
    trailBone {boneName=dyn_coat_bk_r_06}
    trailRelaxationSpeed {editorCurveData {keys [<at 0 15 RCIM_Linear>,<at 1 6 RCIM_Linear>]}}
   }
   n54 = Trail controller - Bone: dyn_coat_r_06!(ComponentPose=@n53, Alpha=1, RelaxationSpeedScale:float=2, FakeVelocity:Vector="0,0,0") -> (Pose) {
    bShowTrailLocation=true
    bUsePlanarLimit=true
    chainLength=3
    debugLifeTime=0.1
    lastBoneRotationAnimAlphaBlend=1
    maxDeltaTime=0.0333
    relaxationSpeedScale=1.2
    alphaScaleBiasClamp {bInterpResult=true, interpSpeedDecreasing=25, interpSpeedIncreasing=25}
    rotationLimits [{limitMax=<180 180 180>}, {limitMax=<180 180 180>}, {limitMax=<180 180 180>}]
    rotationOffsets [<0 0 0>, <0 0 0>, <0 0 0>]
    trailBone {boneName=dyn_coat_r_06}
    trailRelaxationSpeed {editorCurveData {keys [<at 0 15 RCIM_Linear>,<at 1 6 RCIM_Linear>]}}
   }
   n55 = "Constraint - dyn_belt_r_MSTR (PPPPPPPP)"(
    ComponentPose=@n54
    Alpha=1
    ConstraintWeights_0=0.241
    ConstraintWeights_1=0.2059
    ConstraintWeights_2=0.1856
    ConstraintWeights_3=0.1383
    ConstraintWeights_4=0.097
    ConstraintWeights_5=0.0649
    ConstraintWeights_6=0.0375
    ConstraintWeights_7=0.0297
   ) {
    alphaBoolBlend {blendOption=Linear}
    boneToModify {boneName=dyn_belt_r_MSTR}
    constraintSetup [
     {targetBone {boneName=pelvis}}
     {targetBone {boneName=dyn_coat_r_03}}
     {targetBone {boneName=spine_01}}
     {targetBone {boneName=dyn_coat_r_02}}
     {targetBone {boneName=spine_02}}
     {targetBone {boneName=dyn_coat_r_01}}
     {targetBone {boneName=spine_03}}
     {targetBone {boneName=dyn_coat_bk_r_02}}
    ]
    constraintWeights [1,1,1,1,1,1,1,1]
   }
   n56 = "Constraint - dyn_belt_l_MSTR (PPPPPPP)"(
    ComponentPose=@n55
    Alpha=1
    ConstraintWeights_0=0.2394
    ConstraintWeights_1=0.2063
    ConstraintWeights_2=0.1922
    ConstraintWeights_3=0.1402
    ConstraintWeights_4=0.1043
    ConstraintWeights_5=0.07
    ConstraintWeights_6=0.0475
   ) {
    alphaBoolBlend {blendOption=Linear}
    boneToModify {boneName=dyn_belt_l_MSTR}
    constraintSetup [
     {targetBone {boneName=pelvis}}
     {targetBone {boneName=dyn_coat_l_03}}
     {targetBone {boneName=spine_01}}
     {targetBone {boneName=dyn_coat_l_02}}
     {targetBone {boneName=spine_02}}
     {targetBone {boneName=dyn_coat_l_01}}
     {targetBone {boneName=spine_03}}
    ]
    constraintWeights [1,1,1,1,1,1,1]
   }
   n57 = Look At - Bone: dyn_belt_l_01(ComponentPose=@n56, Alpha=1) {
    bUseLookUpAxis=true
    alphaBoolBlend {blendOption=Linear}
    boneToModify {boneName=dyn_belt_l_01}
    lookAtLocation=<-9.3472 3.6832 15.1208>
    lookAtTarget {boneReference {boneName=dyn_coat_l_05}}
    lookAt_Axis {axis=<1 0>}
    lookUp_Axis {axis {x=0, z=1}}
   }
   Save cached pose 'dbnoCache'(Pose=Component To Local(ComponentPose=@n57)) {cacheName=dbnoCache}
   n58 = Get GravityOverride()
   n59 = RigidBody!(ComponentPose=Local To Component(LocalPose=Use cached pose 'StartupCache'()), Alpha, OverrideWorldGravity:Vector=@n58, ExternalForce) -> (Pose) {
    bClampLinearTranslationLimitToRefPose=true
    bOverrideWorldGravity=true
    overridePhysicsAsset=/BRCosmetics/Characters/Player/Male/Medium/Bodies/M_MED_Gimmick/Meshes/M_MED_Gimmick_Dyn_coat_for_cloth
    simulationSpace=BaseBoneSpace
    alphaBoolBlend {blendOption=Linear}
    baseBoneRef {boneName=spine_01}
    componentAppliedLinearAccClamp=<3000 3000 3000>
    componentLinearAccScale=<0.4 0.2 0.2>
    componentLinearVelScale=<0.4 0.2 0.2>
    overrideWorldGravity {z=-980}
    simSpaceSettings {
     maxAngularAcceleration=1000
     maxAngularVelocity=3
     maxLinearAcceleration=700
     maxLinearVelocity=500
     velocityScaleZ=0
     worldAlpha=1
     externalLinearDragV=<0.3 0.3>
    }
   }
   n60 = "Constraint - dyn_belt_r_MSTR (PPPPPPPP)"(
    ComponentPose=@n59
    Alpha=1
    ConstraintWeights_0=0.241
    ConstraintWeights_1=0.2059
    ConstraintWeights_2=0.1856
    ConstraintWeights_3=0.1383
    ConstraintWeights_4=0.097
    ConstraintWeights_5=0.0649
    ConstraintWeights_6=0.0375
    ConstraintWeights_7=0.0297
   ) {
    alphaBoolBlend {blendOption=Linear}
    boneToModify {boneName=dyn_belt_r_MSTR}
    constraintSetup [
     {targetBone {boneName=pelvis}}
     {targetBone {boneName=dyn_coat_r_03}}
     {targetBone {boneName=spine_01}}
     {targetBone {boneName=dyn_coat_r_02}}
     {targetBone {boneName=spine_02}}
     {targetBone {boneName=dyn_coat_r_01}}
     {targetBone {boneName=spine_03}}
     {targetBone {boneName=dyn_coat_bk_r_02}}
    ]
    constraintWeights [1,1,1,1,1,1,1,1]
   }
   n61 = "Constraint - dyn_belt_l_MSTR (PPPPPPP)"(
    ComponentPose=@n60
    Alpha=1
    ConstraintWeights_0=0.2394
    ConstraintWeights_1=0.2063
    ConstraintWeights_2=0.1922
    ConstraintWeights_3=0.1402
    ConstraintWeights_4=0.1043
    ConstraintWeights_5=0.07
    ConstraintWeights_6=0.0475
   ) {
    alphaBoolBlend {blendOption=Linear}
    boneToModify {boneName=dyn_belt_l_MSTR}
    constraintSetup [
     {targetBone {boneName=pelvis}}
     {targetBone {boneName=dyn_coat_l_03}}
     {targetBone {boneName=spine_01}}
     {targetBone {boneName=dyn_coat_l_02}}
     {targetBone {boneName=spine_02}}
     {targetBone {boneName=dyn_coat_l_01}}
     {targetBone {boneName=spine_03}}
    ]
    constraintWeights [1,1,1,1,1,1,1]
   }
   n62 = RigidBody(ComponentPose=@n61, OverrideWorldGravity=@n58) {
    bClampLinearTranslationLimitToRefPose=true
    bOverrideWorldGravity=true
    overridePhysicsAsset=/BRCosmetics/Characters/Player/Male/Medium/Bodies/M_MED_Gimmick/Meshes/M_MED_Gimmick_Dyn_belt_for_cloth
    simulationSpace=BaseBoneSpace
    alphaBoolBlend {blendOption=Linear}
    baseBoneRef {boneName=spine_01}
    componentAppliedLinearAccClamp=<3000 3000 3000>
    componentLinearAccScale=<0.4 0.2 0.2>
    componentLinearVelScale=<0.4 0.2 0.2>
    overrideWorldGravity {z=-980}
    simSpaceSettings {
     maxAngularAcceleration=1000
     maxAngularVelocity=3
     maxLinearAcceleration=700
     maxLinearVelocity=500
     velocityScaleZ=0
     worldAlpha=1
     externalLinearDragV=<0.1 0.1>
    }
   }
   Save cached pose 'RbanCache'(Pose=Component To Local(ComponentPose=@n62)) {cacheName=RbanCache}
   Use cached pose 'StartupCache'()
   Get TrailControllerRot() -> (TrailControllerRot:Rotator)
  ]}
 EventGraph {nodes [
   n1 = Event BlueprintUpdateAnimation()
   > Set DeltaTimeLocal(DeltaTimeLocal:double=@n1.DeltaTimeX) -> (Output_Get:double)
   > n2 = Sequence!(execute) -> (then_0, then_1, then_2:Exec)
   > n3 = Branch(Condition=Get bClothEnabled())
   > n4 = Sequence!(execute) -> (then_0, then_1)
   n5 = "Gravity Override State Function (Parameter Asset) (Nativized)"(
    execute=[@n4.then_0, @n3.else]
    CharacterPartAnimInstance=Self-Reference()
    GameTimeInSeconds=GetGameTimeInSeconds()
    Parameters=/BRCosmetics/Characters/Player/Male/Medium/Bodies/M_MED_Gimmick/Meshes/M_MED_Gimmick_Parameters
   )
   > Set GravityOverride(GravityOverride=@n5)
   n6 = "Cloth Controller Function (Parameter Asset) (Nativized) DEPRECATED"!(
    execute=@n4.then_1
    CharacterPartAnimInstance:Custom Character Part Anim Instance Object Reference=Self-Reference()
    ClothingAssetName:Name=Gimmick_Cloth
    ClothCoreSettings:"Cloth Core Settings Structure (by ref)"=Get Cloth Core Settings Native() -> (Cloth Core Settings Native:Cloth Core Settings Structure)
    WindGustLocalSettings:"Wind Gust Settings Structure (by ref)"=Get Wind Gust Local Settings Native() -> (Wind Gust Local Settings Native:Wind Gust Settings Structure)
    ClothUpAxis:Vector="1,0,0"
    OnGroundMovingSpeedLimit:float=2
    OnGroundMovingAngularVelMax:float=0.05
    OnGroundMovingFictitiousVelMax:float=0.6
    LinVelScaleFwdBwdClamp:float=0
    LinVelScaleUpDwnClamp:float=0.7
    DeltaTime:float=Get DeltaTimeLocal() -> (DeltaTimeLocal:double)
    GameTimeInSeconds:float=GetGameTimeInSeconds()
    StateLogicLayerName:Name=none
    Parameters:Fort Character Dynamics Parameters Object Reference=&M_MED_Gimmick_Parameters
   ) -> (
    then
    OutCurrentState:Name=None
    OutClothParameters:Cloth Parameters Structure
    OutClothCoreSettings:Cloth Core Settings Structure
    OutWindGustSettings:Wind Gust Parameters Structure
    OutWindGustLocalSettings:Wind Gust Settings Structure
    OutWindVelocity:Vector
    OutLinearVelVector:Vector
   )
   > Set Wind Gust Local Settings Native Out(Wind Gust Local Settings Native Out:Wind Gust Settings Structure=@n6.OutWindGustLocalSettings) -> (Output_Get:Wind Gust Settings Structure)
   n7 = PlayerMovement(execute=@n2.then_1, AnimInstance=Self-Reference(), XDotProductCoefficient=1, YDotProductCoefficient=1, ZDotProductCoefficient=1)
   n8 = MapRangeClamped(Value=@n7.YMovement, InRangeA=0, InRangeB=1)
   n9 = "Absolute (Float)"(A=@n7)
   > Set BwdMovement(BwdMovement:double=MapRangeClamped(Value=float + float(A=@n9, B=@n8) {is=title K2Node_CommutativeAssociativeBinaryOperator}, InRangeA=0, InRangeB=1.4)) -> (Output_Get:double)
   n10 = Branch(execute=@n2.then_2, Condition=Get bIsSwimSprinting() -> (bIsSwimSprinting:bool))
   > Set TrailControllerRot(TrailControllerRot:Rotator="15,0, 0") -> (Output_Get:Rotator)
   n11 = Branch(execute=@n10.else, Condition=Get bIsSkyDiving())
   > n12 = Branch(Condition=Get bIsParachuteOpened())
   > Set TrailControllerRot(TrailControllerRot="45,0, 0")
   n13 = Branch(execute=@n12.else, Condition="float > float"(A=Get PawnVelocityZ(), B=50) {is=title K2Node_CallFunction})
   > Set BwdMovement(BwdMovement=float * float(A=Get BwdMovement(), B=0))
   > Set TrailControllerRot(TrailControllerRot="45,0, 0")
   n14 = Branch(execute=@n13.else, Condition=Get bIsSkydiveDiveMode())
   > Set TrailControllerRot(TrailControllerRot="15,0, 0")
   Set TrailControllerRot(execute=@n14.else, TrailControllerRot="60,0, 0")
   Set TrailControllerRot(execute=@n11.else, TrailControllerRot="0, 0, 0")
   > Set BwdMovement(BwdMovement=float * float(A=Get BwdMovement(), B=1))
   Event OnBeginSkydiving()
   > Set TrailControllerRot(TrailControllerRot="-65,0,0")
   Event OnEndSkydiving()
   > Set TrailControllerRot(TrailControllerRot="-10,0,0")
   DrawVectorFromJointFunction!(
    execute
    SkelMeshComponent:Scene Component Object Reference
    JointName:Name=pelvis
    LineEnd:Vector=Break Cloth Setting Struct!(ClothSettingStruct:"Cloth Setting Struct Structure (by ref)") -> (
     LinearVelocityScale_7_#1:Vector=~
     AngularVelocityScale_6_#2:float=~
     FictitiousAngularScale_10_#3:float=~
     GravityOverride_15_#4:Vector=~
     DragCoefficient_23_#5:Vector 2D Structure
     LiftCoefficient_24_#6:Vector 2D Structure
     JointName_27_#7:Name
    ) {is=title K2Node_BreakStruct}.GravityOverride_15_#4
    JointOffset:Vector="0, 0, 0"
    LineColor:Linear Color Structure=(R=1,G=0,B=0,A=1)
    Thickness:double=2
    Foreground:EDrawDebugSceneDepthPriorityGroup Enum=Foreground
   ) -> (then)
  ]}
]
variables [
 BwdMovement {type=number}
 Cloth Core Settings Native {
  type=ClothCoreSettings
  default {
   collisionThickness=0
   dampingCoefficient=0.1
   frictionCoefficient=0.3
   localDampingCoefficient=0.1
   numberIterations=1
   numberSubsteps=1
   selfCollisionThickness=0
   animDriveDamping=<0 1>
   animDriveStiffness=<0 1>
   areaStiffness=<1 1>
   bendingStiffness=<1 1>
   edgeStiffness=<1 1>
   tetherStiffness=<1 1>
  }
 }
 DeltaTimeLocal {type=number}
 GravityOverride {type=Vector}
 TrailControllerRot {type=Rotator}
 Wind Gust Local Settings Native {type=WindGustSettings, default {elapsedTime=0, finalGustStrengthLocal=0, gustDuration=0, gustStrength=0, timeBetweenGusts=0}}
 Wind Gust Local Settings Native Out {type=WindGustSettings, default {elapsedTime=0, finalGustStrengthLocal=0, gustDuration=0, gustStrength=0, timeBetweenGusts=0}}
]
